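# Niagara effect for a hand charge-up in the RoseBruin cosmetic companion's emote.
# One merged ChargeUp emitter drives two mesh and three sprite renderers, tagging particles by spawn group for visibility and selecting per-group attributes through MergeHelper switches.
# Two bursts plus a spawn rate that cuts off two-thirds through the loop feed the groups.
# Curves animate size, color, crystal and ring material parameters, and a flickering pink front glow, with crystal alpha adjusted by time of day.
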NiagaraSystem /CosmeticCompanions/FX/Builtins/RoseBruin/Emote/NS_Emote_RoseBruin_HandChargeUp
dependencies {
 usedDataInterfaces [&NiagaraDataInterfaceCurve, &NiagaraDataInterfaceColorCurve, &NiagaraDataInterfaceSkeletalMesh]
 usedDynamicInputs []
 usedModules [
  &self:SystemSpawnScript.SystemScriptSource.SystemScriptGraph.NiagaraNodeAssignment_0.SetVariables_#1
  &SystemState
  &self:SystemSpawnScript.SystemScriptSource.SystemScriptGraph.NiagaraNodeAssignment_1.SetVariables_#2
 ]
}
emitters {ChargeUp {
  emitterUpdateScript [
   &EmitterState(Life Cycle Mode=::NewEnumerator1, Loop Behavior=::NewEnumerator1, Loop Duration=1.3, UseLoopDelay=false)
   &SpawnBurst_Instantaneous(Use Spawn Probability=false, UseLoopCountLimit=false) {module=SpawnBurst_Instantaneous001}
   &SpawnRate(
    SpawnGroup=1
    SpawnRate=&Multiply_Float(
     A=10
     B=&FloatFromCurve(FloatCurve=DataInterface() {propertyValues {Curve {keys [<at 0.66 1>,<at 0.661 0>]}}}, CurveIndex=$Emitter.NormalizedLoopAge)
    )
    Use Spawn Probability=false
   )
   &SpawnBurst_Instantaneous(Spawn Group=2, Use Spawn Probability=false, UseLoopCountLimit=false)
  ]
  particleSpawnScript [
   /SetVariables(Particles.VisibilityTag=&SpawnGroup()) {isSetParameters=true}
   /SetVariables(Particles.Front Glow Cam Offset=7, Particles.Other Sprites Cam Offset=5) {isSetParameters=true}
   &FN_Mesh_Attributes(
    Create Orientation Attribute=true
    Orientation:Quat4f=&OrientMeshToVector(Look At Direction=<0 0 1>, Reference Axis Vector=<1 0 0>)
    Write to Module Namespace=true
   )
   &InitializeParticle(
    Color=&MergeHelper_ColorSwitch(0=<rgba 1 1 1 1>, 1=<rgba 1 1 1 1>, 2=<rgba 1 0 0.3004 0.6>, 3=?, Index=&SpawnGroup(), Num=::NewEnumerator1)
    Lifetime=&MergeHelper_FloatSwitch(0=1.2, 1=0.5, 2=1.2, 3=?, Index=&SpawnGroup(), Num=::NewEnumerator1)
    Mesh Scale Mode=::NewEnumerator3
    Mesh Uniform Scale=&MergeHelper_FloatSwitch(0=0.25, 1=0.185, 2=0, 3=?, Index=&SpawnGroup(), Num=::NewEnumerator1)
    Sprite Size Mode=::NewEnumerator3
    Uniform Sprite Size=32.5
    UsePositionOffset=false
    Write Material Random=true
   )
  ]
  particleUpdateScript [
   &ParticleState()
   /SetVariables(Particles.Flicker:NiagaraFloat=&Waveform(
     "Waveform [1]"=::NewEnumerator0
     "[1] Frequency"=&RandomRangeFloat(Minimum=4, Maximum=6)
     "[1] Amplitude Min/Max"=<0.4 1.2>
    )) {isSetParameters=true}
   &ScaleColor(
    Scale Alpha=&MergeHelper_FloatSwitch(
     0=1
     1=&FloatFromCurve(
      FloatCurve=DataInterface() {propertyValues {Curve {keys [{arriveTangent=-2.4409, arriveTangentWeight=0.162, interpMode=RCIM_Cubic, leaveTangentWeight=0.2268, time=0.2, value=1}, <at 1 0>]}}}
      Scale Curve=0.4
     )
     2=&Multiply_Float(
      A=0.6
      B=&FloatFromCurve(FloatCurve=DataInterface() {propertyValues {Curve {keys [
           {interpMode=RCIM_Cubic, leaveTangent=2.5294, tangentMode=RCTM_Break, time=0, value=0}
           {arriveTangent=-0.2166, interpMode=RCIM_Cubic, tangentMode=RCTM_Break, time=0.4, value=1}
          ]}}})
     )
     3=?
     Index=&SpawnGroup()
     Num=::NewEnumerator1
    )
    Scale RGB=&MergeHelper_VectorSwitch(
     0=<1 1 1>
     1=&MakeVectorFromLinearColorRGB(Color=&ColorFromCurve(ColorCurve=DataInterface() {propertyValues {
         AlphaCurve {keys []}
         BlueCurve {keys [
           <at 0.0546 1.1 RCIM_Cubic>
           <at 0.1849 1.9662>
           <at 0.351 1.5506>
           <at 0.5954 1.25>
           {arriveTangent=-0.6422, interpMode=RCIM_Cubic, leaveTangent=-0.6422, time=0.7973, value=1}
           <at 0.9846 1>
          ]}
         GreenCurve {keys [
           <at 0.0546 0.3341 RCIM_Cubic>
           <at 0.1849 1.3151>
           <at 0.351 0.7812>
           <at 0.5954 0>
           {arriveTangent=0.358, interpMode=RCIM_Cubic, leaveTangent=0.358, time=0.7973, value=0.7889}
           <at 0.9846 0.1394>
          ]}
         RedCurve {keys [
           <at 0.0546 1.5 RCIM_Cubic>
           <at 0.1849 2.5>
           <at 0.351 2>
           <at 0.5954 0.423>
           {arriveTangent=-1.0865, interpMode=RCIM_Cubic, leaveTangent=-1.0865, time=0.7973, value=0}
           <at 0.9846 0>
          ]}
        }}))
     2=&Lerp_Vector(
      A=&VectorFromFloat(Value=&FloatFromCurve(
        FloatCurve=DataInterface() {propertyValues {Curve {keys [<at 0 1>,<at 0.1 1>,<at 0.2 2>,<at 0.3 1>,<at 0.6 1>,<at 0.7 2>,<at 0.8 1>,<at 1 1>]}}}
        Scale Curve=2
       ))
      B=&VectorFromFloat(Value=&RandomRangeFloat(Minimum=0.3, Maximum=5, Evaluation Type=::NewEnumerator1))
      Alpha=&FloatFromCurve(FloatCurve=DataInterface() {propertyValues {Curve {keys [
           {interpMode=RCIM_Cubic, leaveTangent=2.5294, tangentMode=RCTM_Break, time=0, value=0}
           {arriveTangent=-0.2166, interpMode=RCIM_Cubic, tangentMode=RCTM_Break, time=0.4, value=1}
          ]}}})
     )
     3=?
     Index=&SpawnGroup()
     Num=::NewEnumerator1
    )
    ScaleA=true
    ScaleRGB=true
   )
   /SetVariables(Particles.Crystal Color:LinearColor=&ColorFromCurve(ColorCurve=DataInterface() {propertyValues {
       AlphaCurve {keys [<at 0 1>,<at 0.974 0>]}
       BlueCurve {keys [<at 0.2079 1>,<at 0.4813 1>,<at 0.7278 0.8733>]}
       GreenCurve {keys [<at 0.2079 0>,<at 0.4813 0.1187>,<at 0.7278 0>]}
       RedCurve {keys [<at 0.2079 0.4528>,<at 0.4813 0.0365>,<at 0.7278 1>]}
      }})) {isSetParameters=true}
   /SetVariables(Particles.Scale Crystal Color:LinearColor=&MakeLinearColorFromVectorAndFloat(
     "Vector (RGB)"=&Multiply_Vector(
      A=&MakeVectorFromLinearColorRGB(Color=$Particles.Crystal Color)
      B=&VectorFromFloat(Value=&FloatFromCurve(FloatCurve=DataInterface() {propertyValues {Curve {keys [
            {interpMode=RCIM_Cubic, tangentMode=RCTM_User, time=-0.0056, value=1.6235}
            {interpMode=RCIM_Cubic, tangentMode=RCTM_User, time=0.5, value=3}
            {tangentMode=RCTM_User, time=1, value=0}
           ]}}}))
     )
     "Float (Alpha)"=&Multiply_Float(A=2, B=&FNTimeOfDayScalar(Night Scale=0.5, Morning Scale=1, Evening Scale=1))
    )) {isSetParameters=true}
   &ScaleSpriteSize(Uniform Curve Sprite Scale=DataInterface() {propertyValues {Curve {keys [
        {arriveTangent=7.7185, interpMode=RCIM_Cubic, leaveTangent=7.7185, tangentMode=RCTM_User, time=0.005, value=0.5294}
        {arriveTangent=-5.7154, interpMode=RCIM_Cubic, leaveTangent=-21.3115, tangentMode=RCTM_Break, time=0.5654, value=0.3061}
        {arriveTangent=28.8379, interpMode=RCIM_Cubic, leaveTangent=31.9512, tangentMode=RCTM_Break, time=0.5787, value=0.3235}
        {arriveTangent=-1.007, interpMode=RCIM_Cubic, leaveTangent=-1.007, time=0.6222, value=1.2941}
        {arriveTangent=-0.7091, interpMode=RCIM_Cubic, leaveTangent=-0.7091, tangentMode=RCTM_User, time=0.9, value=0}
       ]}}})
   &ScaleMeshSize(Scale Factor=&VectorFromFloat(Value=&MergeHelper_FloatSwitch(
      0=&FloatFromCurve(FloatCurve=DataInterface() {propertyValues {Curve {keys [<at 0 0.5>,<at 0.15 1>,<at 1 1.2>]}}})
      1=&MakeFloatFromVector(
       Vector=&VectorFromFloat(Value=&FloatFromCurve(FloatCurve=DataInterface() {propertyValues {Curve {
            defaultValue=1
            keys [
             <at 0 1 RCIM_Cubic>
             {arriveTangent=-0.0016, interpMode=RCIM_Cubic, leaveTangent=-0.0016, time=0.0238, value=1.312}
             {arriveTangent=0.7712, interpMode=RCIM_Cubic, leaveTangent=0.7712, tangentMode=RCTM_User, time=0.0582, value=0.9999}
             {arriveTangent=-1.4029, interpMode=RCIM_Cubic, leaveTangent=-1.4029, tangentMode=RCTM_User, time=0.5138, value=0.5514}
             {arriveTangent=0.0846, interpMode=RCIM_Cubic, leaveTangent=0.0846, tangentMode=RCTM_User, time=1, value=0}
            ]
           }}}))
       Channel=::NewEnumerator0
      )
      2=0
      3=?
      Index=&SpawnGroup()
      Num=::NewEnumerator1
     )))
   /SetVariables(Particles.Crystal Dynamic Param:Vector4f=&MakeVector4(
     X=&FloatFromCurve(FloatCurve=DataInterface() {propertyValues {Curve {defaultValue=0.2, keys [{tangentMode=RCTM_User, time=0, value=1}, {tangentMode=RCTM_User, time=1, value=0}]}}})
     Y=&FloatFromCurve(
      FloatCurve=DataInterface() {propertyValues {Curve {keys [{arriveTangent=-0.5818, interpMode=RCIM_Cubic, leaveTangent=-0.5818, tangentMode=RCTM_User, time=0, value=0.5}, <at 1 0.5>]}}}
      Scale Curve=0.5
     )
     Z=&FloatFromCurve(FloatCurve=DataInterface() {propertyValues {Curve {keys [<at 0 0>,<at 0.15 1>,<at 1 1.2>]}}}, Scale Curve=0.5)
     W=0
    )) {isSetParameters=true}
   /SetVariables(Particles.Ring Dynamic Param:Vector4f=&MakeVector4(
     X=&FloatFromCurve(FloatCurve=DataInterface() {propertyValues {Curve {keys [
          {arriveTangent=-7.2529, interpMode=RCIM_Cubic, leaveTangent=-7.2529, tangentMode=RCTM_User, time=-0.0022, value=1.0993}
          {arriveTangent=0.875, interpMode=RCIM_Cubic, leaveTangent=0.875, tangentMode=RCTM_User, time=0.5, value=0}
          {arriveTangent=3.8908, interpMode=RCIM_Cubic, leaveTangent=3.8908, tangentMode=RCTM_User, time=1, value=1}
         ]}}})
     Y=&FloatFromCurve(FloatCurve=DataInterface() {propertyValues {Curve {keys [
          {arriveTangent=5.2096, interpMode=RCIM_Cubic, leaveTangent=5.2096, tangentMode=RCTM_User, time=0, value=0.5}
          {arriveTangent=0.2132, interpMode=RCIM_Cubic, leaveTangent=0.2132, tangentMode=RCTM_User, time=1, value=3}
         ]}}})
     Z=&FloatFromCurve(FloatCurve=DataInterface() {propertyValues {Curve {
         defaultValue=1
         keys [{arriveTangent=2.1224, interpMode=RCIM_Cubic, leaveTangent=2.1224, tangentMode=RCTM_User, time=0, value=0}, <at 1 1 RCIM_Cubic>]
        }}})
     W=&FloatFromCurve(
      FloatCurve=DataInterface() {propertyValues {Curve {keys [
          {arriveTangent=2.3563, interpMode=RCIM_Cubic, leaveTangent=2.3563, tangentMode=RCTM_User, time=0, value=0}
          {tangentMode=RCTM_User, time=1, value=1}
         ]}}}
      Scale Curve=2
     )
    )) {isSetParameters=true}
   /SetVariables(
    Particles.Front Glow Color:LinearColor=&MakeLinearColorFromVectorAndFloat(
     "Vector (RGB)"=&MakeVectorFromLinearColorRGB(Color=<rgba 2 0.1021 0.5271 0.3826>)
     "Float (Alpha)"=&FloatFromCurve(
      FloatCurve=DataInterface() {propertyValues {Curve {
         defaultValue=1
         keys [
          {arriveTangent=3.6783, interpMode=RCIM_Cubic, leaveTangent=3.6783, tangentMode=RCTM_User, time=0, value=-0.0127}
          {interpMode=RCIM_Cubic, tangentMode=RCTM_User, time=0.1458, value=0.3862}
          {arriveTangent=2.3497, interpMode=RCIM_Cubic, leaveTangent=2.3497, time=0.3191, value=0.4633}
          {arriveTangent=-0.1422, interpMode=RCIM_Cubic, leaveTangent=-0.1422, time=0.4473, value=1.0947}
          {arriveTangent=-2.5243, interpMode=RCIM_Cubic, leaveTangent=-2.5243, time=0.5393, value=0.432}
          {arriveTangent=-0.9377, interpMode=RCIM_Cubic, leaveTangent=-0.9377, time=0.6527, value=0.5764}
          {tangentMode=RCTM_User, time=1, value=0}
         ]
        }}}
      Scale Curve=&Multiply_Float(A=$Particles.Flicker, B=1)
     )
    )
    Particles.Front Glow Size:Vector2f=&Vector2DFromFloat(Value=40)
   ) {isSetParameters=true}
   /SetVariables(Particles.Linear Color=&MultiplyLinearColor(
     Linear Color=<rgba 1 0 0.3004 0.6>
     Scale Factor=&MakeVector4_FromLinearColor(LinearColor=&MakeLinearColorFromVectorAndFloat(
       "Vector (RGB)"=&Lerp_Vector(
        A=&VectorFromFloat(Value=&FloatFromCurve(
          FloatCurve=DataInterface() {propertyValues {Curve {keys [<at 0 1>,<at 0.1 1>,<at 0.2 2>,<at 0.3 1>,<at 0.6 1>,<at 0.7 2>,<at 0.8 1>,<at 1 1>]}}}
          Scale Curve=2
         ))
        B=&VectorFromFloat(Value=&RandomRangeFloat(Minimum=0.3, Maximum=5, Evaluation Type=::NewEnumerator1))
        Alpha=&FloatFromCurve(FloatCurve=DataInterface() {propertyValues {Curve {keys [
             {interpMode=RCIM_Cubic, leaveTangent=2.5294, tangentMode=RCTM_Break, time=0, value=0}
             {arriveTangent=-0.2166, interpMode=RCIM_Cubic, tangentMode=RCTM_Break, time=0.4, value=1}
            ]}}})
       )
       "Float (Alpha)"=&FloatFromCurve(FloatCurve=DataInterface() {propertyValues {Curve {keys [
            {interpMode=RCIM_Cubic, leaveTangent=2.5294, tangentMode=RCTM_Break, time=0, value=0}
            {arriveTangent=-0.2166, interpMode=RCIM_Cubic, tangentMode=RCTM_Break, time=0.4, value=1}
           ]}}})
      ))
    )) {isSetParameters=true}
  ]
  rendererClasses [&NiagaraMeshRendererProperties, &NiagaraSpriteRendererProperties]
  renderers [
   {rendererClass=&NiagaraMeshRendererProperties, rendererIndex=0}
   {rendererClass=&NiagaraMeshRendererProperties, rendererIndex=1}
   {rendererClass=&NiagaraSpriteRendererProperties, rendererIndex=2}
   {rendererClass=&NiagaraSpriteRendererProperties, rendererIndex=3}
   {rendererClass=&NiagaraSpriteRendererProperties, rendererIndex=4}
  ]
 }}
summary {
 systemName=NS_Emote_RoseBruin_HandChargeUp
 emitters [{emitterName=ChargeUp, rendererClasses [&NiagaraMeshRendererProperties, &NiagaraSpriteRendererProperties]}]
 userVariables [
  User.bIsFrontEnd {type=NiagaraBool}
  User.bIsFrontEndPreview {type=NiagaraBool}
 ]
}
