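# Niagara effect for the RoseBruin cosmetic companion's hover emote, drawing pink thruster-like glows and sparks on skeletal mesh bones or sockets.
# Fountain streams short-lived pink-to-purple sprites that fall, inherit velocity and scatter with curl noise.
# SingleLoopingParticle bursts two spawn groups, re-pins them every update with per-partition offsets, and layers six flare renderers with waveform flicker.
# User.FX_Curve, compared against zero, switches the "Iron Man" glow sizes to zero.
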NiagaraSystem /CosmeticCompanions/FX/Builtins/RoseBruin/Emote/NS_Emote_RoseBruin_Hover
dependencies {
 usedDataInterfaces [&NiagaraDataInterfaceCurve, &NiagaraDataInterfaceSkeletalMesh]
 usedDynamicInputs []
 usedModules [
  &SystemState
  &self:SystemSpawnScript.SystemScriptSource.SystemScriptGraph.NiagaraNodeAssignment_0.SetVariables_#1
 ]
}
emitters {
 Fountain {
  emitterUpdateScript [
   &EmitterState(Life Cycle Mode=::NewEnumerator1, Loop Duration=2, UseLoopDelay=false)
   &SpawnRate(SpawnRate=75, Use Spawn Probability=false)
  ]
  particleSpawnScript [
   &PartitionParticles()
   &InitializeParticle(
    Color Maximum=<rgba 1.25 1.1213 1.1836 1>
    Color Minimum=<rgba 2 0.2181 0.9258 1>
    Color Mode=::NewEnumerator2
    Lifetime Max=0.4
    Lifetime Min=0.15
    Lifetime Mode=::NewEnumerator1
    Mass Mode=::NewEnumerator3
    Sprite Rotation Mode=::NewEnumerator2
    Sprite Size Mode=::NewEnumerator1
    Uniform Sprite Size Max=12
    Uniform Sprite Size Min=2
    UsePositionOffset=false
    Write Material Random=true
   )
   &SkeletalMeshLocation(
    Bone / Socket Index=$Particles.PartitionParticles.Partition
    Bone Sampling Mode=::NewEnumerator4
    MaskSpawnGroup=false
    Sampled Position Offset=&MaskVector(Mask Source=1, Mask Parameter=$Particles.PartitionParticles.Partition, Mask=0, Masked Vector=<2 0 -3>, Passthrough Vector=<-2 0 -3>)
    Sampled Position Offset Coordinate Space:ENiagaraSkelMeshTransforms=::NewEnumerator2
    Skeletal Mesh=$System.Skeletal Mesh
   )
   &ShapeLocation(MaskSpawnGroup=false, Override Seed=false, Sphere Radius=8, UseOwnerScale=true)
   &AddVelocity(Velocity=<0 0 -20>)
  ]
  particleUpdateScript [
   &ParticleState()
   &ScaleColor(
    Scale Alpha=&FloatFromCurve(FloatCurve=DataInterface() {propertyValues {Curve {keys [<at -0.0046 1.0113 RCIM_Cubic>,<at 1.0011 0.002 RCIM_Cubic>]}}})
    Scale RGB=&Multiply_Vector(
     A=<1 1 1>
     B=&MakeVectorFromLinearColorRGB(Color=&LerpLinearColors(
       StartColor=<rgba 1 1 1 1>
       EndColor=&RandomRangeLinearColor(Minimum=<rgba 1 0 0.2718 1>, Maximum=<rgba 0.6118 0 1 1>)
       LerpFactor=&FloatFromCurve(FloatCurve=DataInterface() {propertyValues {Curve {defaultValue=1, keys [<at 0 0 RCIM_Cubic>,<at 0.75 1 RCIM_Cubic>]}}})
      ))
    )
    ScaleA=true
    ScaleRGB=true
   )
   &InheritVelocity(Inherited Velocity Amount Scale=<0.5 0.5 0.5>, Speed Limit=true, UseSpeedThreshold=false)
   &CurlNoiseForce(
    Noise Strength=&FloatFromCurve(FloatCurve=DataInterface() {propertyValues {Curve {defaultValue=1, keys [<at 0.05 0 RCIM_Cubic>,<at 0.6 1 RCIM_Cubic>]}}}, Scale Curve=100)
    Pan Noise=false
   )
   &Drag(Drag=&RandomRangeFloat(Minimum=0.2, Maximum=2), Use Linear Drag=true, Use Rotational Drag=false)
   &SolveForcesAndVelocity(Clamp Velocity=false, Limit Acceleration=false)
  ]
  rendererClasses [&NiagaraSpriteRendererProperties]
  renderers [{rendererClass=&NiagaraSpriteRendererProperties}]
 }
 SingleLoopingParticle {
  emitterUpdateScript [
   &EmitterState()
   &SpawnBurst_Instantaneous(Spawn Count=2, Use Spawn Probability=false, UseLoopCountLimit=false)
   &SpawnBurst_Instantaneous(Spawn Count=2, Spawn Group=1, Use Spawn Probability=false, UseLoopCountLimit=false) {module=SpawnBurst_Instantaneous001}
  ]
  particleSpawnScript [
   /SetVariables(Particles.VisibilityTag=&SpawnGroup()) {isSetParameters=true}
   &PartitionParticles()
   &InitializeParticle(
    Color=<rgba 1 0.0648 0.449 0.1>
    Lifetime=&MaskFloat(Mask Source=0, Mask Parameter=?, Mask=0, Masked Float=$Emitter.CurrentLoopDuration, Passthrough Float=0.88)
    Sprite Size Mode=::NewEnumerator3
    Uniform Sprite Size=25
    UsePositionOffset=false
    Write Material Random=true
   )
   &CameraOffset(Camera Offset Amount=10)
  ]
  particleUpdateScript [
   &ParticleState()
   /SetVariables(
    Particles.Flicker:NiagaraFloat=&Waveform(
     "Waveform [1]"=::NewEnumerator0
     "[1] Frequency"=&RandomRangeFloat(Minimum=4, Maximum=8)
     "[1] Amplitude Min/Max"=<0.3 1.2>
    )
    Particles.Iron Man Size Flicker:NiagaraFloat=&Waveform("Waveform [1]"=::NewEnumerator0, "[1] Frequency"=5, "[1] Amplitude Min/Max"=<0.2 1.25>)
   ) {isSetParameters=true}
   &SolveForcesAndVelocity(Clamp Velocity=false, Limit Acceleration=false)
   &SkeletalMeshLocation(
    Bone / Socket Index=$Particles.PartitionParticles.Partition
    Bone Sampling Mode=::NewEnumerator4
    MaskSpawnGroup=false
    Sampled Position Offset=&MergeHelper_VectorSwitch(
     0=&MaskVector(Mask Source=1, Mask Parameter=$Particles.PartitionParticles.Partition, Mask=0, Masked Vector=<2 5 -3.7>, Passthrough Vector=<-2 5 -3.7>)
     1=&MaskVector(Mask Source=1, Mask Parameter=$Particles.PartitionParticles.Partition, Mask=0, Masked Vector=<0 0 -4.2>, Passthrough Vector=<0 0 -4.2>)
     2=?
     3=?
     Index=&SpawnGroup()
     Num=::NewEnumerator0
    )
    Sampled Position Offset Coordinate Space:ENiagaraSkelMeshTransforms=::NewEnumerator2
    Skeletal Mesh=$System.Skeletal Mesh
   )
   /SetVariables(
    Particles.Corona Color:LinearColor=&MakeLinearColorFromVectorAndFloat(
     "Vector (RGB)"=&MakeVectorFromLinearColorRGB(Color=<rgba 0.5567 0.0354 1 1>)
     "Float (Alpha)"=&FloatFromCurve(
      FloatCurve=DataInterface() {propertyValues {Curve {
         defaultValue=1
         keys [
          {interpMode=RCIM_Cubic, tangentMode=RCTM_User, time=0, value=0}
          {interpMode=RCIM_Cubic, tangentMode=RCTM_User, time=0.1401, value=1.0194}
          {arriveTangent=-1.4775, interpMode=RCIM_Cubic, leaveTangent=-1.4775, time=0.354, value=0.0687}
          {arriveTangent=0.0846, interpMode=RCIM_Cubic, leaveTangent=0.0846, time=0.455, value=0.5543}
          {arriveTangent=0.7378, interpMode=RCIM_Cubic, leaveTangent=0.7378, time=0.5858, value=0.0883}
          {arriveTangent=-0.2132, interpMode=RCIM_Cubic, leaveTangent=-0.2132, time=0.7626, value=0.7813}
          {tangentMode=RCTM_User, time=1, value=0}
         ]
        }}}
      Scale Curve=&Multiply_Float(A=1, B=$Particles.Flicker)
     )
    )
    Particles.Corona Size:Vector2f=<20 20>
    Particles.CoronaFlare Color:LinearColor=&MakeLinearColorFromVectorAndFloat(
     "Vector (RGB)"=&MakeVectorFromLinearColorRGB(Color=<rgba 1.25 0.125 0.8163 3>)
     "Float (Alpha)"=&FloatFromCurve(
      FloatCurve=DataInterface() {propertyValues {Curve {
         defaultValue=1
         keys [
          {interpMode=RCIM_Cubic, tangentMode=RCTM_User, time=0, value=0}
          {interpMode=RCIM_Cubic, tangentMode=RCTM_User, time=0.1458, value=0.3862}
          {arriveTangent=2.2101, interpMode=RCIM_Cubic, leaveTangent=2.2101, time=0.354, value=0.0687}
          {arriveTangent=0.0846, interpMode=RCIM_Cubic, leaveTangent=0.0846, time=0.4606, value=1.082}
          {arriveTangent=-1.8921, interpMode=RCIM_Cubic, leaveTangent=-1.8921, time=0.5858, value=0.0883}
          {arriveTangent=-0.2132, interpMode=RCIM_Cubic, leaveTangent=-0.2132, time=0.8287, value=0.3855}
          {tangentMode=RCTM_User, time=1, value=0}
         ]
        }}}
      Scale Curve=&Multiply_Float(A=1, B=$Particles.Flicker)
     )
    )
    Particles.CoronaFlare Size:Vector2f=<35 35>
    Particles.SharpFlare Color:LinearColor=&MakeLinearColorFromVectorAndFloat(
     "Vector (RGB)"=&MakeVectorFromLinearColorRGB(Color=<rgba 1 0.3997 0.6996 1>)
     "Float (Alpha)"=&FloatFromCurve(
      FloatCurve=DataInterface() {propertyValues {Curve {
         defaultValue=1
         keys [
          {interpMode=RCIM_Cubic, tangentMode=RCTM_User, time=0, value=0}
          {interpMode=RCIM_Cubic, tangentMode=RCTM_User, time=0.1698, value=0.5277}
          {arriveTangent=0.0932, interpMode=RCIM_Cubic, leaveTangent=0.0932, time=0.354, value=0.0687}
          {arriveTangent=0.0846, interpMode=RCIM_Cubic, leaveTangent=0.0846, time=0.455, value=0.5543}
          {arriveTangent=1.4005, interpMode=RCIM_Cubic, leaveTangent=1.4005, time=0.5858, value=0.0883}
          {arriveTangent=-0.2132, interpMode=RCIM_Cubic, leaveTangent=-0.2132, time=0.7626, value=0.9851}
          {tangentMode=RCTM_User, time=1, value=0}
         ]
        }}}
      Scale Curve=&Multiply_Float(A=1, B=$Particles.Flicker)
     )
    )
    Particles.SharpFlareSize:Vector2f=<30 25>
   ) {isSetParameters=true}
   /SetVariables(
    Particles.Iron Man 01 Size:Vector2f=&MakeCustomVector2DFromBool(
     Bool=&SetBoolByFloatComparison(A=$User.FX_Curve, B=0)
     True Vector2D=<0 0>
     False Vector2D=&Multiply_Vector2D_ByFloat(
      Float=$Particles.Iron Man Size Flicker
      Vector2D=&Multiply_Vector2D(
       A=<30 25>
       B=&Vector2DFromFloat(Value=&FloatFromCurve(
         FloatCurve=DataInterface() {propertyValues {Curve {
            defaultValue=1
            keys [
             <at 0 1 RCIM_Cubic>
             {arriveTangent=0.912, interpMode=RCIM_Cubic, leaveTangent=0.912, tangentMode=RCTM_User, time=0.5036, value=1.2496}
             {arriveTangent=5.1199, interpMode=RCIM_Cubic, leaveTangent=5.1199, time=0.8976, value=2.8226}
             <at 0.9971 3.7766 RCIM_Cubic>
            ]
           }}}
         Scale Curve=0.7
        ))
      )
     )
    )
    Particles.Iron Man Color:LinearColor=&MakeLinearColorFromVectorAndFloat(
     "Vector (RGB)"=&MakeVectorFromLinearColorRGB(Color=<rgba 5 1.4097 2.1801 1>)
     "Float (Alpha)"=&FloatFromCurve(FloatCurve=DataInterface() {propertyValues {Curve {
         defaultValue=1
         keys [
          <at -0.0012 0 RCIM_Cubic>
          {arriveTangent=3.0831, interpMode=RCIM_Cubic, leaveTangent=3.0831, time=0.1675, value=0.3366}
          {arriveTangent=0.5902, interpMode=RCIM_Cubic, leaveTangent=0.5902, tangentMode=RCTM_User, time=0.3231, value=1}
          {arriveTangent=-1.5327, interpMode=RCIM_Cubic, leaveTangent=-1.5327, time=0.5805, value=0.3151}
          {interpMode=RCIM_Cubic, tangentMode=RCTM_User, time=0.9756, value=0}
         ]
        }}})
    )
    Particles.Iron Man Size:Vector2f=&MakeCustomVector2DFromBool(
     Bool=&SetBoolByFloatComparison(A=$User.FX_Curve, B=0)
     True Vector2D=<0 0>
     False Vector2D=&Multiply_Vector2D_ByFloat(
      Float=$Particles.Iron Man Size Flicker
      Vector2D=&Multiply_Vector2D(
       A=<50 45>
       B=&Vector2DFromFloat(Value=&FloatFromCurve(
         FloatCurve=DataInterface() {propertyValues {Curve {
            defaultValue=1
            keys [
             <at 0 1 RCIM_Cubic>
             {arriveTangent=0.912, interpMode=RCIM_Cubic, leaveTangent=0.912, tangentMode=RCTM_User, time=0.5036, value=1.2496}
             {arriveTangent=5.1199, interpMode=RCIM_Cubic, leaveTangent=5.1199, time=0.8976, value=2.8226}
             <at 0.9971 3.7766 RCIM_Cubic>
            ]
           }}}
         Scale Curve=0.7
        ))
      )
     )
    )
   ) {isSetParameters=true}
  ]
  rendererClasses [&NiagaraSpriteRendererProperties]
  renderers [
   {rendererClass=&NiagaraSpriteRendererProperties}
   {rendererClass=&NiagaraSpriteRendererProperties, rendererIndex=1}
   {rendererClass=&NiagaraSpriteRendererProperties, rendererIndex=2}
   {rendererClass=&NiagaraSpriteRendererProperties, rendererIndex=3}
   {rendererClass=&NiagaraSpriteRendererProperties, rendererIndex=4}
   {rendererClass=&NiagaraSpriteRendererProperties, rendererIndex=5}
  ]
 }
}
summary {
 systemName=NS_Emote_RoseBruin_Hover
 emitters [
  {emitterName=SingleLoopingParticle, rendererClasses [&NiagaraSpriteRendererProperties]}
  {emitterName=Fountain, rendererClasses [&NiagaraSpriteRendererProperties]}
 ]
 userVariables [User.FX_Curve {type=NiagaraFloat, defaultValue=0, description=""}]
}
