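# Niagara effect for the RoseBruin cosmetic companion emote, building a glowing pink heart that bursts into sparkles and flares about three seconds in.
# A Beam emitter adds a short mesh-and-sprite flash, while Condensed_Meshes draws layered heart meshes.
# Condensed_Sprites packs glows, flickering sparkles and burst debris into one emitter, using spawn groups as visibility tags to select per-layer attributes.
# Some brightness values adapt to platform, time of day or front end.
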NiagaraSystem /CosmeticCompanions/FX/Builtins/RoseBruin/Emote/NS_Emote_RoseBruin_Trails
dependencies {
 usedDataInterfaces [
  &NiagaraDataInterfaceCurve
  &NiagaraDataInterfaceArrayColor
  &NiagaraDataInterfaceColorCurve
  &NiagaraDataInterfaceParticleRead
  &NiagaraDataInterfacePlatformSet
  &NiagaraDataInterfaceSkeletalMesh
 ]
 usedDynamicInputs []
 usedModules [
  &SystemState
  &self:SystemSpawnScript.SystemScriptSource.SystemScriptGraph.NiagaraNodeAssignment_1.SetVariables_#1
 ]
}
emitters {
 Beam {
  emitterUpdateScript [
   &EmitterState(Inactive Response=::NewEnumerator1, Life Cycle Mode=::NewEnumerator1, Loop Behavior=::NewEnumerator1, UseLoopDelay=false)
   &SpawnBurst_Instantaneous(Spawn Time=3, Use Spawn Probability=false, UseLoopCountLimit=false)
   /SetVariables(Emitter.SplineWidth_Base=1) {isSetParameters=true}
   /SetVariables(Emitter.SplineWidth_Tip=1) {isSetParameters=true}
   /SetVariables(Emitter.SplineWidthGlow_Base=1) {isSetParameters=true}
   /SetVariables(Emitter.SplineWidthRed_Tip=15) {isSetParameters=true}
   /SetVariables(Emitter.Linear Color:LinearColor=&ColorFromCurve(
     ColorCurve=DataInterface() {propertyValues {
       AlphaCurve {keys []}
       BlueCurve {keys [
         <at 0.5632 12.4968>
         {arriveTangent=-76.713, interpMode=RCIM_Cubic, leaveTangent=-76.713, time=0.6275, value=8.2016}
         <at 0.6801 2.6361>
         {interpMode=RCIM_Cubic, leaveTangent=-20.6997, time=0.7196, value=0.5038}
        ]}
       GreenCurve {keys [
         <at 0.5632 8.9945>
         {arriveTangent=-57.1763, interpMode=RCIM_Cubic, leaveTangent=-57.1763, time=0.6275, value=3.2827}
         <at 0.6801 0.6361>
         {interpMode=RCIM_Cubic, leaveTangent=-8.6728, time=0.7196, value=0.0559}
        ]}
       RedCurve {keys [
         <at 0.5632 20>
         {arriveTangent=-121.5334, interpMode=RCIM_Cubic, leaveTangent=-121.5334, time=0.6275, value=20}
         <at 0.6801 8.9232>
         {interpMode=RCIM_Cubic, leaveTangent=-51.0774, time=0.7196, value=1}
        ]}
      }}
     CurveIndex=$Emitter.NormalizedLoopAge
    )) {isSetParameters=true}
  ]
  particleSpawnScript [
   &InitializeParticle(
    Color Mode=::NewEnumerator0
    Lifetime=0.64
    Mesh Scale=<0.4 0.4 1.35>
    Mesh Scale Mode=::NewEnumerator4
    Position Offset=<0 0 100>
    Sprite Size=<112 94>
    Sprite Size Mode=::NewEnumerator4
    UsePositionOffset=true
    Write Material Random=true
   )
   &InitialMeshOrientation(Mesh Axis To Orient=<0 0 1>, Mesh Orientation Mode=::NewEnumerator2, Orientation Vector=<0 0 -1>, Use Rotation Vector=true)
   &FN_Sprite_Attributes(Camera Offset=2, Create Camera Offset Attribute=true, Write to Module Namespace=true)
  ]
  particleUpdateScript [
   &ParticleState()
   &ScaleSpriteSize(Uniform Curve Sprite Scale=DataInterface() {propertyValues {Curve {
       defaultValue=0
       keys [
        {arriveTangent=53.9492, interpMode=RCIM_Cubic, leaveTangent=53.9492, tangentMode=RCTM_User, time=0, value=0}
        {arriveTangent=-4.3323, interpMode=RCIM_Cubic, leaveTangent=-4.3323, tangentMode=RCTM_User, time=0.1408, value=0.9847}
        {arriveTangent=-7.2478, interpMode=RCIM_Cubic, leaveTangent=-7.2478, tangentMode=RCTM_User, time=1, value=0}
       ]
      }}})
   &ScaleMeshSize(Scale Factor=&Multiply_Vector(
     A=&MakeVector(
      X=&FloatFromCurve(
       FloatCurve=DataInterface() {propertyValues {Curve {keys [
           {arriveTangent=1533.2161, interpMode=RCIM_Cubic, leaveTangent=1533.2161, tangentMode=RCTM_User, time=0.0041, value=15.4359}
           {
            arriveTangent=-85.9264
            arriveTangentWeight=0.0433
            interpMode=RCIM_Cubic
            leaveTangent=0.7957
            leaveTangentWeight=0.4281
            tangentMode=RCTM_Break
            time=0.1083
            value=1
           }
           {arriveTangent=2.6833, interpMode=RCIM_Cubic, leaveTangent=2656.8723, tangentMode=RCTM_Break, time=0.1504, value=1}
           {arriveTangent=-292.6025, interpMode=RCIM_Cubic, leaveTangent=-292.6025, tangentMode=RCTM_User, time=0.3983, value=61.1789}
           {arriveTangent=4.4735, interpMode=RCIM_Cubic, leaveTangent=4.4735, time=0.5325, value=28.218}
           {arriveTangent=-31.6575, interpMode=RCIM_Cubic, leaveTangent=-31.6575, tangentMode=RCTM_User, time=0.9982, value=63.8627}
          ]}}}
       Scale Curve=1.35
      )
      Y=1
      Z=&FloatFromCurve(
       FloatCurve=DataInterface() {propertyValues {Curve {keys [
           {arriveTangent=16.0145, interpMode=RCIM_Cubic, leaveTangent=16.0145, tangentMode=RCTM_User, time=-0.0005, value=0}
           {arriveTangent=17.8946, interpMode=RCIM_Cubic, leaveTangent=17.8946, tangentMode=RCTM_User, time=0.0396, value=0.6097}
           {arriveTangent=-32.3419, interpMode=RCIM_Cubic, leaveTangent=1.9836, tangentMode=RCTM_Break, time=0.0929, value=1}
           {arriveTangent=0.4938, interpMode=RCIM_Cubic, leaveTangent=0.4938, tangentMode=RCTM_User, time=0.9384, value=1}
           <at 1.0063 0 RCIM_Cubic>
          ]}}}
       Scale Curve=1.2
      )
     )
     B=&Divide_Vector(A:Vector3f=<1 1 1>, B:Vector3f=$Engine.Owner.Scale)
    )) {enabled=false}
   &ScaleMeshSize(Scale Factor=&Multiply_Vector(
     A=&MakeVector(
      X=&FloatFromCurve(
       FloatCurve=DataInterface() {propertyValues {Curve {keys [
           {interpMode=RCIM_Cubic, leaveTangentWeight=1.7894, tangentMode=RCTM_User, time=0.0001, value=14.7571}
           {arriveTangent=-45.5382, interpMode=RCIM_Cubic, leaveTangent=-45.5382, time=0.1469, value=14.8459}
           {arriveTangentWeight=13.4088, interpMode=RCIM_Cubic, leaveTangentWeight=1.8278, tangentMode=RCTM_User, time=0.1902, value=6.1}
           {arriveTangent=-18.9027, interpMode=RCIM_Cubic, leaveTangent=-18.9027, time=0.2023, value=51.7096}
           {arriveTangentWeight=5.5781, interpMode=RCIM_Cubic, leaveTangentWeight=10.0336, tangentMode=RCTM_User, time=0.2261, value=5.4217}
           {arriveTangent=116.1605, interpMode=RCIM_Cubic, leaveTangent=116.1605, time=0.2515, value=13.9681}
           {arriveTangent=2.6833, time=0.2968, value=13.6316}
           {arriveTangent=694.2539, time=0.3021, value=95.2376}
           <at 0.332 84.6269>
           <at 0.3466 140.5109>
           <at 0.3596 74.7133>
           <at 0.4037 66.7492>
           {arriveTangent=-65.5456, interpMode=RCIM_Cubic, leaveTangent=-65.5456, time=0.4663, value=49.9432}
           {arriveTangent=-31.6575, interpMode=RCIM_Cubic, leaveTangent=-31.6575, tangentMode=RCTM_User, time=0.5516, value=57.0535}
           {arriveTangent=-54.8799, interpMode=RCIM_Cubic, leaveTangent=-54.8799, tangentMode=RCTM_User, time=0.6453, value=56.1665}
           {arriveTangent=-125.8564, interpMode=RCIM_Cubic, leaveTangent=-125.8564, time=0.7621, value=27.4068}
           <at 1.0013 11.3516 RCIM_Cubic>
          ]}}}
       Scale Curve=1.35
      )
      Y=1
      Z=&FloatFromCurve(FloatCurve=DataInterface() {propertyValues {Curve {keys [
           {arriveTangent=4.2318, interpMode=RCIM_Cubic, leaveTangent=4.2318, tangentMode=RCTM_User, time=-0.0005, value=0}
           {arriveTangent=19.2818, interpMode=RCIM_Cubic, leaveTangent=19.2818, tangentMode=RCTM_User, time=0.0654, value=0.4222}
           {arriveTangent=-764.8372, interpMode=RCIM_Cubic, leaveTangent=1.9836, tangentMode=RCTM_Break, time=0.1129, value=0.9464}
           {arriveTangent=0.4938, interpMode=RCIM_Cubic, leaveTangent=0.4938, tangentMode=RCTM_User, time=0.9241, value=1.3639}
           <at 1.0063 0 RCIM_Cubic>
          ]}}})
     )
     B=&Divide_Vector(A:Vector3f=<1 1 1>, B:Vector3f=$Engine.Owner.Scale)
    )) {module=ScaleMeshSize001}
   /SetVariables(Particles.Glow Color=&MakeLinearColorFromVectorAndFloat(
     "Vector (RGB)"=&MakeVectorFromLinearColorRGB(Color=<rgba 1 0.5837 0.7509 0.4174>)
     "Float (Alpha)"=&FloatFromCurve(FloatCurve=DataInterface() {propertyValues {Curve {
         defaultValue=1
         keys [
          {arriveTangent=81.536, interpMode=RCIM_Cubic, leaveTangent=81.536, tangentMode=RCTM_User, time=0, value=0}
          {arriveTangentWeight=0.0139, interpMode=RCIM_Cubic, leaveTangentWeight=0.2559, tangentMode=RCTM_User, time=0.0416, value=1.0194}
          {interpMode=RCIM_Cubic, tangentMode=RCTM_User, time=0.5014, value=0}
         ]
        }}})
    )) {isSetParameters=true}
   /SetVariables(Particles.Energy Color:LinearColor=&MakeLinearColorFromVectorAndFloat(
     "Vector (RGB)"=&MakeVectorFromLinearColorRGB(Color=<rgba 1.5 1.5 1.5 1>)
     "Float (Alpha)"=&Multiply_Float(
      A=&FloatFromCurve(
       FloatCurve=DataInterface() {propertyValues {Curve {keys [<at 0 5>,<at 0.0803 2.6471>,<at 0.426 0.8824>,<at 1 0>]}}}
       Scale Curve=&FNTimeOfDayScalar(Night Scale=0.6, Morning Scale=0.8, Evening Scale=0.6)
      )
      B=&FN_PickFloatDependOnPlatform(TargetPlatformActive=1, TargetPlatformNoActive=0.15)
     )
    )) {isSetParameters=true}
   /SetVariables(Particles.Center Color:LinearColor=&MakeLinearColorFromVectorAndFloat(
     "Vector (RGB)"=&MakeVectorFromLinearColorRGB(Color=<rgba 1 1 1 1.5>)
     "Float (Alpha)"=&Multiply_Float(
      A=&FloatFromCurve(
       FloatCurve=DataInterface() {propertyValues {Curve {keys [<at 0 1>,<at 0.7805 1.0118>,<at 1 0>]}}}
       Scale Curve=&FNTimeOfDayScalar(Night Scale=0.6, Morning Scale=0.8, Evening Scale=0.6)
      )
      B=&FN_PickFloatDependOnPlatform(TargetPlatformActive=1, TargetPlatformNoActive=0.6)
     )
    )) {isSetParameters=true}
   &DynamicMaterialParameters(
    Index 0 Param 1=&FloatFromCurve(
     FloatCurve=DataInterface() {propertyValues {Curve {keys [
         {arriveTangent=12.0832, interpMode=RCIM_Cubic, leaveTangent=12.0832, tangentMode=RCTM_User, time=0.7, value=-1}
         {arriveTangent=0.7491, interpMode=RCIM_Cubic, leaveTangent=0.7491, tangentMode=RCTM_User, time=1, value=1}
        ]}}}
     Scale Curve=&RandomRangeFloat(Minimum=0.9, Maximum=1)
    )
    Index 0 Param 2=1
    Index 0 Param 3=1
    Index 0 Param 4=100
    Module.Index0Param0Write=true
    Module.Index0Param1Write=true
    Module.Index0Param2Write=true
    Module.Index0Param3Write=true
   )
   &ScaleColor(
    Scale Alpha=&FloatFromCurve(FloatCurve=DataInterface() {propertyValues {Curve {
        defaultValue=1
        keys [
         {arriveTangent=36.5539, interpMode=RCIM_Cubic, leaveTangent=36.5539, tangentMode=RCTM_User, time=0.0456, value=-0.0192}
         {arriveTangentWeight=0.0166, interpMode=RCIM_Cubic, leaveTangentWeight=0.111, tangentMode=RCTM_User, time=0.0952, value=0.9808}
         {
          arriveTangent=-1.525
          arriveTangentWeight=0.1611
          interpMode=RCIM_Cubic
          leaveTangent=-1.525
          leaveTangentWeight=0.2987
          tangentMode=RCTM_User
          time=0.5024
          value=0.5968
         }
         {arriveTangent=-0.0405, interpMode=RCIM_Cubic, leaveTangent=-0.0405, tangentMode=RCTM_User, time=0.9496, value=0}
        ]
       }}})
    Scale RGB=&MakeVectorFromLinearColorRGB(Color=&ColorFromCurve(ColorCurve=DataInterface() {propertyValues {
        AlphaCurve {keys []}
        BlueCurve {keys [
          <at 0.012 3.7857>
          <at 0.1042 1.5624>
          {arriveTangent=-3.0834, interpMode=RCIM_Cubic, leaveTangent=-3.0834, time=0.4287, value=1.5624}
          <at 1 0.7393 RCIM_Cubic>
         ]}
        GreenCurve {keys [
          <at 0.012 2.8976>
          <at 0.1042 0.4522>
          {arriveTangent=-2.9328, interpMode=RCIM_Cubic, leaveTangent=-2.9328, time=0.4287, value=0.4522}
          <at 1 0 RCIM_Cubic>
         ]}
        RedCurve {keys [
          <at 0.012 5>
          <at 0.1042 5>
          {arriveTangent=-3.6437, interpMode=RCIM_Cubic, leaveTangent=-3.6437, time=0.4287, value=5}
          <at 1 1.4 RCIM_Cubic>
         ]}
       }}))
    ScaleA=true
    ScaleRGB=true
   )
   /SetVariables(
    Particles.Big Flare Color:LinearColor=&MakeLinearColorFromVectorAndFloat(
     "Vector (RGB)"=&MakeVectorFromLinearColorRGB(Color=&ColorFromCurve(ColorCurve=DataInterface() {propertyValues {
         AlphaCurve {keys []}
         BlueCurve {keys [<at 0 1.5624 RCIM_Cubic>, {interpMode=RCIM_Cubic, leaveTangent=-4.0222, time=0.0638, value=0.2923}, <at 0.1398 0.2923>]}
         GreenCurve {keys [<at 0 0.4522 RCIM_Cubic>, {interpMode=RCIM_Cubic, leaveTangent=-1.1874, time=0.0638, value=0}, <at 0.1398 0>]}
         RedCurve {keys [<at 0 5 RCIM_Cubic>, {interpMode=RCIM_Cubic, leaveTangent=-12.8965, time=0.0638, value=1}, <at 0.1398 1>]}
        }}))
     "Float (Alpha)"=&FloatFromCurve(
      FloatCurve=DataInterface() {propertyValues {Curve {
         defaultValue=1
         keys [
          {arriveTangent=36.5539, interpMode=RCIM_Cubic, leaveTangent=36.5539, tangentMode=RCTM_User, time=0.0456, value=-0.0192}
          {arriveTangentWeight=0.0166, interpMode=RCIM_Cubic, leaveTangentWeight=0.111, tangentMode=RCTM_User, time=0.0952, value=0.9808}
          {
           arriveTangent=-1.525
           arriveTangentWeight=0.1611
           interpMode=RCIM_Cubic
           leaveTangent=-1.525
           leaveTangentWeight=0.2987
           tangentMode=RCTM_User
           time=0.5024
           value=0.5968
          }
          {arriveTangent=-0.0405, interpMode=RCIM_Cubic, leaveTangent=-0.0405, tangentMode=RCTM_User, time=0.9496, value=0}
         ]
        }}}
      Scale Curve=0.6
     )
    )
    Particles.Big Flare Size:Vector2f=&Multiply_Vector2D_ByFloat(
     Float=&FloatFromCurve(FloatCurve=DataInterface() {propertyValues {Curve {
         defaultValue=1
         keys [
          {arriveTangent=9.0725, interpMode=RCIM_Cubic, leaveTangent=9.0725, tangentMode=RCTM_User, time=0, value=0}
          {arriveTangent=-0.0703, interpMode=RCIM_Cubic, leaveTangent=-0.0703, tangentMode=RCTM_User, time=0.1408, value=0.9847}
          {arriveTangent=-2.4264, interpMode=RCIM_Cubic, leaveTangent=-2.4264, tangentMode=RCTM_User, time=1, value=0}
         ]
        }}})
     Vector2D=<140 100>
    )
   ) {isSetParameters=true}
  ]
  rendererClasses [&NiagaraMeshRendererProperties, &NiagaraSpriteRendererProperties]
  renderers [
   {rendererClass=&NiagaraMeshRendererProperties, rendererIndex=0}
   {rendererClass=&NiagaraMeshRendererProperties, rendererIndex=1}
   {rendererClass=&NiagaraMeshRendererProperties, rendererIndex=2}
   {rendererClass=&NiagaraSpriteRendererProperties, rendererIndex=3}
   {rendererClass=&NiagaraSpriteRendererProperties, rendererIndex=4}
  ]
 }
 Condensed_Meshes {
  emitterUpdateScript [
   &EmitterState(Life Cycle Mode=::NewEnumerator1, Loop Behavior=::NewEnumerator1, Loop Duration=7, UseLoopDelay=false)
   &SpawnBurst_Instantaneous(Spawn Time=0.3, Use Spawn Probability=false, UseLoopCountLimit=false)
  ]
  particleSpawnScript [
   /SetVariables(Particles.VisibilityTag=&SpawnGroup()) {isSetParameters=true}
   /SetVariables(Particles.Lifetimes=3.75) {isSetParameters=true}
   &InitializeParticle(
    Color=<rgba 2.5 0.1824 1.091 1>
    Lifetime=$Particles.Lifetimes
    Mesh Scale Mode=::NewEnumerator3
    Mesh Uniform Scale=1.12
    Position Offset=&MaskVector(Mask Source=0, Mask Parameter=?, Mask=0, Masked Vector=<0 0 4.7>, Passthrough Vector=<0 0 38>)
    UsePositionOffset=true
    Write Material Random=true
   )
   /SetVariables(Particles.Highlight Position:NiagaraPosition=&AddVectorToPosition(Position=$Particles.Position, Vector=<0 0 -2.5>)) {isSetParameters=true}
  ]
  particleUpdateScript [
   &ParticleState()
   /SetVariables(Particles.HeartMesh Color:LinearColor=&MakeLinearColorFromVectorAndFloat(
     "Vector (RGB)"=&MakeVectorFromLinearColorRGB(Color=<rgba 2.5 0.1824 1.091 1>)
     "Float (Alpha)"=&FloatFromCurve(FloatCurve=DataInterface() {propertyValues {Curve {
         defaultValue=1
         keys [
          {interpMode=RCIM_Cubic, tangentMode=RCTM_User, time=0.0593, value=0}
          {arriveTangent=0.8712, interpMode=RCIM_Cubic, leaveTangent=0.8712, tangentMode=RCTM_User, time=0.2427, value=1}
          {arriveTangent=-1.018, interpMode=RCIM_Cubic, leaveTangent=-1.018, tangentMode=RCTM_User, time=0.8391, value=0.993}
          {arriveTangent=-2.4881, interpMode=RCIM_Cubic, leaveTangent=-2.4881, tangentMode=RCTM_User, time=0.9945, value=0}
         ]
        }}})
    )) {isSetParameters=true}
   /SetVariables(Particles.HeartMesh01 Color:LinearColor=&MakeLinearColorFromVectorAndFloat(
     "Vector (RGB)"=&MakeVectorFromLinearColorRGB(Color=<rgba 10 1.1315 5.528 0.8>)
     "Float (Alpha)"=&FloatFromCurve(
      FloatCurve=DataInterface() {propertyValues {Curve {
         defaultValue=1
         keys [
          {interpMode=RCIM_Cubic, tangentMode=RCTM_User, time=0.0593, value=0}
          {arriveTangent=0.8712, interpMode=RCIM_Cubic, leaveTangent=0.8712, tangentMode=RCTM_User, time=0.2427, value=1}
          {arriveTangent=-1.018, interpMode=RCIM_Cubic, leaveTangent=-1.018, tangentMode=RCTM_User, time=0.8391, value=0.993}
          {arriveTangent=-2.4881, interpMode=RCIM_Cubic, leaveTangent=-2.4881, tangentMode=RCTM_User, time=0.9945, value=0}
         ]
        }}}
      Scale Curve=0.3
     )
    )) {isSetParameters=true}
   /SetVariables(
    Particles.HeartBolt Color:LinearColor=&MakeLinearColorFromVectorAndFloat(
     "Vector (RGB)"=&MakeVectorFromLinearColorRGB(Color=<rgba 2.5 0.7541 1.6019 1>)
     "Float (Alpha)"=&FloatFromCurve(FloatCurve=DataInterface() {propertyValues {Curve {defaultValue=1, keys [{tangentMode=RCTM_User, time=0.776, value=0}, {tangentMode=RCTM_User, time=0.8528, value=1.0757}, <at 1.0018 0>]}}})
    )
    Particles.HeartBolt Size:Vector3f=&Multiply_Vector(
     A=$Particles.Initial.Scale
     B=&Multiply_Vector(
      A=&MakeVector(
       X=1
       Y=1
       Z=&FloatFromCurve(FloatCurve=DataInterface() {propertyValues {Curve {keys [
            {arriveTangent=0.5218, interpMode=RCIM_Cubic, leaveTangent=0.5218, tangentMode=RCTM_User, time=0, value=0}
            {
             arriveTangent=-68.3187
             arriveTangentWeight=0.0377
             interpMode=RCIM_Cubic
             leaveTangentWeight=0.3339
             tangentMode=RCTM_Break
             time=0.076
             value=1
            }
           ]}}})
      )
      B=&Divide_Vector(A:Vector3f=<1 1 1>, B:Vector3f=$Engine.Owner.Scale)
     )
    )
   ) {isSetParameters=true}
   /SetVariables(
    Particles.Highlight Color:LinearColor=&MakeLinearColorFromVectorAndFloat(
     "Vector (RGB)"=&MakeVectorFromLinearColorRGB(Color=<rgba 8 3.5243 5.6768 0.1>)
     "Float (Alpha)"=&FloatFromCurve(FloatCurve=DataInterface() {propertyValues {Curve {
         defaultValue=1
         keys [
          {interpMode=RCIM_Cubic, tangentMode=RCTM_User, time=0.0593, value=0}
          {arriveTangent=0.8712, interpMode=RCIM_Cubic, leaveTangent=0.8712, tangentMode=RCTM_User, time=0.2427, value=1}
          {arriveTangent=-1.018, interpMode=RCIM_Cubic, leaveTangent=-1.018, tangentMode=RCTM_User, time=0.8391, value=0.993}
          {arriveTangent=-2.4881, interpMode=RCIM_Cubic, leaveTangent=-2.4881, tangentMode=RCTM_User, time=0.9945, value=0}
         ]
        }}})
    )
    Particles.Highlight Size:Vector3f=&VectorFromFloat(Value=1.2)
   ) {isSetParameters=true}
  ]
  rendererClasses [&NiagaraMeshRendererProperties]
  renderers [
   {rendererClass=&NiagaraMeshRendererProperties, rendererIndex=0}
   {rendererClass=&NiagaraMeshRendererProperties, rendererIndex=1}
   {rendererClass=&NiagaraMeshRendererProperties, rendererIndex=2}
   {rendererClass=&NiagaraMeshRendererProperties, rendererIndex=3}
  ]
 }
 Condensed_Sprites {
  emitterUpdateScript [
   &EmitterState(Life Cycle Mode=::NewEnumerator1, Loop Behavior=::NewEnumerator1, Loop Duration=7, UseLoopDelay=false)
   &SpawnBurst_Instantaneous(Spawn Time=0.3, Use Spawn Probability=false, UseLoopCountLimit=false) {module=SpawnBurst_Instantaneous001}
   &SpawnBurst_Instantaneous(Spawn Group=1, Spawn Time=3.07, Use Spawn Probability=false, UseLoopCountLimit=false) {module=SpawnBurst_Instantaneous002}
   &SpawnBurst_Instantaneous(Spawn Count=5, Spawn Group=2, Use Spawn Probability=false, UseLoopCountLimit=false) {module=SpawnBurst_Instantaneous003}
   &SpawnBurst_Instantaneous(Spawn Count=7, Spawn Group=2, Spawn Time=3.07, Use Spawn Probability=false, UseLoopCountLimit=false) {module=SpawnBurst_Instantaneous004}
   &SpawnRate(
    SpawnGroup=2
    SpawnRate=&Multiply_Float(
     A=20
     B=&FloatFromCurve(
      FloatCurve=DataInterface() {propertyValues {Curve {
         defaultValue=1
         keys [{tangentMode=RCTM_User, time=0.15, value=0}, {tangentMode=RCTM_User, time=0.25, value=1}, <at 0.6013 1>, <at 0.6229 0>]
        }}}
      CurveIndex=$Emitter.NormalizedLoopAge
     )
    )
    Use Spawn Probability=false
   )
   &SpawnBurst_Instantaneous(Spawn Count=15, Spawn Group=3, Spawn Time=3.07, Use Spawn Probability=false, UseLoopCountLimit=false) {module=SpawnBurst_Instantaneous005}
   &SpawnRate(
    SpawnGroup=3
    SpawnRate=&Multiply_Float(
     A=25
     B=&FloatFromCurve(
      FloatCurve=DataInterface() {propertyValues {Curve {
         defaultValue=1
         keys [{tangentMode=RCTM_User, time=0.15, value=0}, {tangentMode=RCTM_User, time=0.25, value=1}, <at 0.61 1.0353>, <at 0.643 0>]
        }}}
      CurveIndex=$Emitter.NormalizedLoopAge
     )
    )
    Use Spawn Probability=false
   ) {module=SpawnRate001}
   &SpawnBurst_Instantaneous(Spawn Count=11, Spawn Group=4, Spawn Time=3.07, Use Spawn Probability=false, UseLoopCountLimit=false) {module=SpawnBurst_Instantaneous006}
   &SpawnBurst_Instantaneous(Spawn Group=5, Spawn Time=3.07, Use Spawn Probability=false, UseLoopCountLimit=false) {module=SpawnBurst_Instantaneous007}
   &SpawnBurst_Instantaneous(Spawn Group=6, Spawn Time=3.07, Use Spawn Probability=false, UseLoopCountLimit=false) {module=SpawnBurst_Instantaneous008}
   &SpawnBurst_Instantaneous(Spawn Group=7, Spawn Time=3.07, Use Spawn Probability=false, UseLoopCountLimit=false) {module=SpawnBurst_Instantaneous009}
   &SpawnBurst_Instantaneous(Spawn Group=8, Spawn Time=3.07, Use Spawn Probability=false, UseLoopCountLimit=false)
   &SpawnBurst_Instantaneous(
    Spawn Count=&RandomRangeInt(Minimum=7, Maximum=13)
    Spawn Group=9
    Spawn Time=3.07
    Use Spawn Probability=false
    UseLoopCountLimit=false
   ) {module=SpawnBurst_Instantaneous010}
   &SpawnBurst_Instantaneous(
    Spawn Count=&RandomRangeInt(Minimum=11, Maximum=15)
    Spawn Group=10
    Spawn Time=3.07
    Use Spawn Probability=false
    UseLoopCountLimit=false
   ) {module=SpawnBurst_Instantaneous011}
   &SpawnBurst_Instantaneous(
    Spawn Count=&RandomRangeInt(Minimum=13, Maximum=17)
    Spawn Group=11
    Spawn Time=3.07
    Use Spawn Probability=false
    UseLoopCountLimit=false
   ) {module=SpawnBurst_Instantaneous012}
  ]
  particleSpawnScript [
   /SetVariables(Particles.VisibilityTag=&SpawnGroup()) {isSetParameters=true}
   /SetVariables(Particles.Lifetimes:NiagaraFloat=&MaskFloat(
     Mask Source=0
     Mask Parameter=?
     Mask=0
     Masked Float=3.75
     Passthrough Float=&MaskFloat(
      Mask Source=0
      Mask Parameter=?
      Mask=1
      Masked Float=0.12
      Passthrough Float=&MaskFloat(
       Mask Source=0
       Mask Parameter=?
       Mask=2
       Masked Float=&RandomRangeFloat(Minimum=1, Maximum=2)
       Passthrough Float=&MaskFloat(
        Mask Source=0
        Mask Parameter=?
        Mask=3
        Masked Float=&RandomRangeFloat(Minimum=0.17, Maximum=0.6)
        Passthrough Float=&MaskFloat(
         Mask Source=0
         Mask Parameter=?
         Mask=4
         Masked Float=&RandomRangeFloat(Minimum=0.3, Maximum=0.6)
         Passthrough Float=&MaskFloat(
          Mask Source=0
          Mask Parameter=?
          Mask=5
          Masked Float=0.15
          Passthrough Float=&MaskFloat(
           Mask Source=0
           Mask Parameter=?
           Mask=6
           Masked Float=0.15
           Passthrough Float=&MaskFloat(
            Mask Source=0
            Mask Parameter=?
            Mask=7
            Masked Float=0.11
            Passthrough Float=&MaskFloat(
             Mask Source=0
             Mask Parameter=?
             Mask=8
             Masked Float=0.25
             Passthrough Float=&MaskFloat(
              Mask Source=0
              Mask Parameter=?
              Mask=9
              Masked Float=&RandomRangeFloat(Minimum=0.12, Maximum=0.6)
              Passthrough Float=&MaskFloat(
               Mask Source=0
               Mask Parameter=?
               Mask=10
               Masked Float=&RandomRangeFloat(Minimum=0.12, Maximum=0.6)
               Passthrough Float=&RandomRangeFloat(Minimum=0.12, Maximum=0.6)
              )
             )
            )
           )
          )
         )
        )
       )
      )
     )
    )) {isSetParameters=true}
   &InitializeParticle(
    Color=&MergeHelper_ColorSwitch(
     0=<rgba 1 0.0214 0.3734 0.45>
     1=<rgba 1.3 0.7075 1.0068 0.8>
     2=&RandomRangeLinearColor(Minimum=<rgba 1 0.2535 0.6847 1>, Maximum=<rgba 0.5417 0 0.3577 1>)
     3=<rgba 0 0 0 0>
     4=<rgba 0 0 0 0>
     5=<rgba 5 1.2687 3.2844 2>
     Index=&SpawnGroup()
     Num=::NewEnumerator4
    )
    Lifetime=$Particles.Lifetimes
    Position Offset=<0 0 38>
    Sprite Size Mode=::NewEnumerator3
    Uniform Sprite Size=&MaskFloat(
     Mask Source=0
     Mask Parameter=?
     Mask=0
     Masked Float=55
     Passthrough Float=&MaskFloat(
      Mask Source=0
      Mask Parameter=?
      Mask=1
      Masked Float=55
      Passthrough Float=&MaskFloat(
       Mask Source=0
       Mask Parameter=?
       Mask=2
       Masked Float=&RandomRangeFloat(Minimum=2, Maximum=17)
       Passthrough Float=&MaskFloat(
        Mask Source=0
        Mask Parameter=?
        Mask=3
        Masked Float=&RandomRangeFloat(Minimum=3, Maximum=15)
        Passthrough Float=&MaskFloat(
         Mask Source=0
         Mask Parameter=?
         Mask=4
         Masked Float=&RandomRangeFloat(Minimum=3, Maximum=15)
         Passthrough Float=&MaskFloat(
          Mask Source=0
          Mask Parameter=?
          Mask=5
          Masked Float=50
          Passthrough Float=&MaskFloat(
           Mask Source=0
           Mask Parameter=?
           Mask=6
           Masked Float=50
           Passthrough Float=&MaskFloat(
            Mask Source=0
            Mask Parameter=?
            Mask=7
            Masked Float=50
            Passthrough Float=&MaskFloat(
             Mask Source=0
             Mask Parameter=?
             Mask=8
             Masked Float=66
             Passthrough Float=&MaskFloat(
              Mask Source=0
              Mask Parameter=?
              Mask=9
              Masked Float=&RandomRangeFloat(Minimum=2, Maximum=15)
              Passthrough Float=&MaskFloat(
               Mask Source=0
               Mask Parameter=?
               Mask=10
               Masked Float=&RandomRangeFloat(Minimum=2, Maximum=13)
               Passthrough Float=&RandomRangeFloat(Minimum=2, Maximum=12)
              )
             )
            )
           )
          )
         )
        )
       )
      )
     )
    )
    UsePositionOffset=true
    Write Material Random=true
   )
   /SetVariables(Particles.Pos Adjustment:NiagaraPosition=&AddVectorToPosition(Position=$Particles.Position, Vector=<0.7 0 0>)) {isSetParameters=true}
   /SetVariables(Particles.Pos Adjustment 01:NiagaraPosition=&AddVectorToPosition(Position=$Particles.Position, Vector=<0.6 0 -0.2>)) {isSetParameters=true}
   /SetVariables(Particles.Heart Line Sprite Size:Vector2f=<46 48>) {isSetParameters=true}
   &ShapeLocation(
    MaskSpawnGroup=false
    Non Uniform Scale=<1 0.1 1.2>
    Offset Mode=::NewEnumerator0
    Override Seed=false
    Sphere Radius=&MaskFloat(
     Mask Source=0
     Mask Parameter=?
     Mask=2
     Masked Float=20
     Passthrough Float=&MaskFloat(
      Mask Source=0
      Mask Parameter=?
      Mask=3
      Masked Float=25
      Passthrough Float=&MaskFloat(
       Mask Source=0
       Mask Parameter=?
       Mask=4
       Masked Float=17
       Passthrough Float=&MaskFloat(
        Mask Source=0
        Mask Parameter=?
        Mask=9
        Masked Float=20
        Passthrough Float=&MaskFloat(
         Mask Source=0
         Mask Parameter=?
         Mask=10
         Masked Float=20
         Passthrough Float=&MaskFloat(Mask Source=0, Mask Parameter=?, Mask=11, Masked Float=20, Passthrough Float=0)
        )
       )
      )
     )
    )
    Sphere Surface Distribution=1
    UseOwnerScale=true
   )
   &AddVelocity(
    Module.Use Falloff Distance=false
    Override Seed=false
    Velocity Mode=::NewEnumerator1
    Velocity Speed=&MaskFloat(
     Mask Source=0
     Mask Parameter=?
     Mask=9
     Masked Float=&RandomRangeFloat(Minimum=75, Maximum=400)
     Passthrough Float=&MaskFloat(
      Mask Source=0
      Mask Parameter=?
      Mask=10
      Masked Float=&RandomRangeFloat(Minimum=75, Maximum=400)
      Passthrough Float=&MaskFloat(
       Mask Source=0
       Mask Parameter=?
       Mask=11
       Masked Float=&RandomRangeFloat(Minimum=75, Maximum=400)
       Passthrough Float=0
      )
     )
    )
   )
   &ApplyInitialForces()
  ]
  particleUpdateScript [
   &ParticleState()
   /SetVariables(Particles.Flicker:NiagaraFloat=&Waveform(
     "Waveform [1]"=::NewEnumerator0
     "[1] Frequency"=&RandomRangeFloat(Minimum=0.2, Maximum=4)
     "[1] Amplitude Min/Max"=&RandomRangeVector2D(Minimum=<0.4 0.6>, Maximum=<0.2 0.8>)
     "[1] Scale"=&RandomRangeFloat(Minimum=0.6, Maximum=1)
    )) {isSetParameters=true}
   &ScaleColor(
    Scale Alpha=&MergeHelper_FloatSwitch(
     0=&FloatFromCurve(FloatCurve=DataInterface() {propertyValues {Curve {
         defaultValue=1
         keys [
          {interpMode=RCIM_Cubic, tangentMode=RCTM_User, time=0.0593, value=0}
          {arriveTangent=0.8712, interpMode=RCIM_Cubic, leaveTangent=0.8712, tangentMode=RCTM_User, time=0.2427, value=1}
          {arriveTangent=-1.018, interpMode=RCIM_Cubic, leaveTangent=-1.018, tangentMode=RCTM_User, time=0.8391, value=0.993}
          {arriveTangent=-2.4881, interpMode=RCIM_Cubic, leaveTangent=-2.4881, tangentMode=RCTM_User, time=0.9945, value=0}
         ]
        }}})
     1=&FloatFromCurve(FloatCurve=DataInterface() {propertyValues {Curve {
         defaultValue=1
         keys [
          {arriveTangent=-2.9302, interpMode=RCIM_Cubic, leaveTangent=-2.9302, tangentMode=RCTM_User, time=0.1011, value=1}
          {interpMode=RCIM_Cubic, tangentMode=RCTM_User, time=1, value=0}
         ]
        }}})
     2=&Multiply_Float(
      A=$Particles.Flicker
      B=&FloatFromCurve(
       FloatCurve=DataInterface() {propertyValues {Curve {
          defaultValue=1.0113
          keys [
           {interpMode=RCIM_Cubic, tangentMode=RCTM_User, time=0, value=0}
           {interpMode=RCIM_Cubic, tangentMode=RCTM_User, time=0.3886, value=1}
           {tangentMode=RCTM_User, time=1, value=0}
          ]
         }}}
       Scale Curve=0.65
      )
     )
     3=0
     4=0
     5=&FloatFromCurve(FloatCurve=DataInterface() {propertyValues {Curve {keys [
          <at -0.0046 1.0113 RCIM_Cubic>
          {arriveTangent=-1.0036, interpMode=RCIM_Cubic, leaveTangent=-1.0036, time=0.0697, value=0.4519}
          <at 1.0011 0.002 RCIM_Cubic>
         ]}}})
     Index=&SpawnGroup()
     Num=::NewEnumerator4
    )
    Scale RGB=&MergeHelper_VectorSwitch(
     0=<1 1 1>
     1=<1 1 1>
     2=&MakeVectorFromLinearColorRGB(Color=&MultiplyLinearColor(
       Linear Color=<rgba 1 1 1 1>
       Scale Factor=&MakeVector4_FromLinearColor(LinearColor=&SelectLinearColorFromArray(
         Array Sampling Mode=::NewEnumerator0
         Direct Array Index=?
         Color Selection Array=DataInterface() {propertyValues {ColorData [<rgba 1 0.8385 0.9451 1>, <rgba 0.64 0.28 1 1>, <rgba 1 0.3594 0.5173 1>, <rgba 0.1198 0 0.118 1>]}}
         Random Seed=0
        ))
      ))
     3=<1 1 1>
     Index=&SpawnGroup()
     Num=::NewEnumerator2
    )
    ScaleA=true
    ScaleRGB=true
   )
   /SetVariables(
    Particles.0 - Heart Glow Color:LinearColor=&MakeLinearColorFromVectorAndFloat(
     "Vector (RGB)"=&MakeVectorFromLinearColorRGB(Color=<rgba 1 0.0214 0.3734 0.45>)
     "Float (Alpha)"=&FloatFromCurve(FloatCurve=DataInterface() {propertyValues {Curve {
         defaultValue=1
         keys [
          {interpMode=RCIM_Cubic, tangentMode=RCTM_User, time=0.0593, value=0}
          {arriveTangent=0.8712, interpMode=RCIM_Cubic, leaveTangent=0.8712, tangentMode=RCTM_User, time=0.2427, value=1}
          {arriveTangent=-1.018, interpMode=RCIM_Cubic, leaveTangent=-1.018, tangentMode=RCTM_User, time=0.8391, value=0.993}
          {arriveTangent=-2.4881, interpMode=RCIM_Cubic, leaveTangent=-2.4881, tangentMode=RCTM_User, time=0.9945, value=0}
         ]
        }}})
    )
    Particles.0 - Heart Line Glow Color:LinearColor=&MakeLinearColorFromVectorAndFloat(
     "Vector (RGB)"=&MakeVectorFromLinearColorRGB(Color=<rgba 1 0.0214 0.3734 0.45>)
     "Float (Alpha)"=&FloatFromCurve(FloatCurve=DataInterface() {propertyValues {Curve {defaultValue=1, keys [{tangentMode=RCTM_User, time=0.776, value=0}, {tangentMode=RCTM_User, time=0.8528, value=1.0757}, <at 1.0018 0>]}}})
    )
    Particles.1 - Heart Burst Color:LinearColor=&MakeLinearColorFromVectorAndFloat(
     "Vector (RGB)"=&MakeVectorFromLinearColorRGB(Color=<rgba 1.3 0.7075 1.0068 0.8>)
     "Float (Alpha)"=&FloatFromCurve(FloatCurve=DataInterface() {propertyValues {Curve {
         defaultValue=1
         keys [
          {arriveTangent=-2.9302, interpMode=RCIM_Cubic, leaveTangent=-2.9302, tangentMode=RCTM_User, time=0.1011, value=1}
          {interpMode=RCIM_Cubic, tangentMode=RCTM_User, time=1, value=0}
         ]
        }}})
    )
    Particles.1 - Heart Line Burst Color:LinearColor=&MakeLinearColorFromVectorAndFloat(
     "Vector (RGB)"=&MakeVectorFromLinearColorRGB(Color=<rgba 1.3 0.7075 1.0068 0.8>)
     "Float (Alpha)"=&FloatFromCurve(FloatCurve=DataInterface() {propertyValues {Curve {
         defaultValue=1
         keys [
          {arriveTangent=-2.9302, interpMode=RCIM_Cubic, leaveTangent=-2.9302, tangentMode=RCTM_User, time=0.1011, value=1}
          {interpMode=RCIM_Cubic, tangentMode=RCTM_User, time=1, value=0}
         ]
        }}})
    )
   ) {isSetParameters=true}
   /SetVariables(Particles.3/4 - Sparkles Color:LinearColor=&MakeLinearColorFromVectorAndFloat(
     "Vector (RGB)"=&MakeVectorFromLinearColorRGB(Color=&LerpLinearColors(
       StartColor=&RandomRangeLinearColor(Minimum=<rgba 2 0 0.4174 1>, Maximum=<rgba 2 0.4817 1.3032 1>)
       EndColor=&RandomRangeLinearColor(Minimum=<rgba 0.1256 0 0.3073 1>, Maximum=<rgba 0.2917 0.0005 0.1378 1>)
       LerpFactor=&Multiply_Float(
        A=$Particles.Flicker
        B=&FloatFromCurve(FloatCurve=DataInterface() {propertyValues {Curve {
            defaultValue=1
            keys [{interpMode=RCIM_Cubic, tangentMode=RCTM_User, time=0.25, value=0}, {interpMode=RCIM_Cubic, tangentMode=RCTM_User, time=0.75, value=1}]
           }}})
       )
      ))
     "Float (Alpha)"=&Multiply_Float(
      A=$Particles.Flicker
      B=&FloatFromCurve(FloatCurve=DataInterface() {propertyValues {Curve {
          defaultValue=1
          keys [
           {interpMode=RCIM_Cubic, tangentMode=RCTM_User, time=0, value=0}
           {interpMode=RCIM_Cubic, tangentMode=RCTM_User, time=0.5, value=1}
           {tangentMode=RCTM_User, time=1, value=0}
          ]
         }}})
     )
    )) {isSetParameters=true}
   /SetVariables(
    Particles.6 - Glow Color:LinearColor=&MakeLinearColorFromVectorAndFloat(
     "Vector (RGB)"=&MakeVectorFromLinearColorRGB(Color=<rgba 2 0.5423 0.37 1>)
     "Float (Alpha)"=&FloatFromCurve(FloatCurve=DataInterface() {propertyValues {Curve {keys [
          <at -0.0046 1.0113 RCIM_Cubic>
          {arriveTangent=-1.0036, interpMode=RCIM_Cubic, leaveTangent=-1.0036, time=0.0697, value=0.4519}
          <at 1.0011 0.002 RCIM_Cubic>
         ]}}})
    )
    Particles.6 - Glow Size:Vector2f=<100 100>
   ) {isSetParameters=true}
   /SetVariables(
    Particles.7 - Dark Glow Color:LinearColor=&MakeLinearColorFromVectorAndFloat(
     "Vector (RGB)"=&MakeVectorFromLinearColorRGB(Color=<rgba 0.1615 0.0014 0.0718 1>)
     "Float (Alpha)"=&FloatFromCurve(
      FloatCurve=DataInterface() {propertyValues {Curve {
         defaultValue=1
         keys [{interpMode=RCIM_Cubic, tangentMode=RCTM_User, time=0, value=1}, {interpMode=RCIM_Cubic, tangentMode=RCTM_User, time=1, value=0}]
        }}}
      Scale Curve=2
     )
    )
    Particles.7 - Dark Glow Size:Vector2f=&Vector2DFromFloat(Value=&Multiply_Float(A=160, B=1))
   ) {isSetParameters=true}
   /SetVariables(
    Particles.8 - Radial Sprite Color:LinearColor=&MakeLinearColorFromVectorAndFloat(
     "Vector (RGB)"=&MakeVectorFromLinearColorRGB(Color=&MultiplyLinearColor(
       Linear Color=<rgba 1 0.5145 0.7765 1>
       Scale Factor=&MakeVector4_FromLinearColor(LinearColor=&ColorFromCurve(ColorCurve=DataInterface() {propertyValues {
           AlphaCurve {keys [<at 0 1 RCIM_Cubic>]}
           BlueCurve {keys [<at 0 1 RCIM_Cubic>,<at 0.5354 1>,<at 0.9851 0.5227>]}
           GreenCurve {keys [<at 0 0.8848 RCIM_Cubic>,<at 0.5354 0.4697>,<at 0.9851 0.8085>]}
           RedCurve {keys [<at 0 0.306 RCIM_Cubic>,<at 0.5354 0.6193>,<at 0.9851 1>]}
          }}))
      ))
     "Float (Alpha)"=&FloatFromCurve(
      FloatCurve=DataInterface() {propertyValues {Curve {keys [{arriveTangent=-3.3955, interpMode=RCIM_Cubic, leaveTangent=-3.3955, tangentMode=RCTM_User, time=0.2, value=1}, <at 1 0 RCIM_Cubic>]}}}
      Scale Curve=&FN_FrontEndFloatModifier(Gameplay=5, Front End=3)
     )
    )
    Particles.8 - Radial Sprite Size:Vector2f=&Vector2DFromFloat(Value=&Multiply_Float(
      A=66
      B=&MakeFloatFromVector2d(Vector 2D=&Vector2DFromFloat(Value=&FloatFromCurve(
         FloatCurve=DataInterface() {propertyValues {Curve {keys [{arriveTangent=1.7733, interpMode=RCIM_Cubic, leaveTangent=1.7733, tangentMode=RCTM_User, time=0, value=0.2}, <at 1 1 RCIM_Cubic>]}}}
         Scale Curve=1.25
        )))
     ))
   ) {isSetParameters=true}
   /SetVariables(
    Particles.10 - Diamond Sprites Color:LinearColor=&MakeLinearColorFromVectorAndFloat(
     "Vector (RGB)"=&MakeVectorFromLinearColorRGB(Color=&MultiplyLinearColor(
       Linear Color=&RandomRangeLinearColor(Minimum=<rgba 2 0.9062 1.4463 1>, Maximum=<rgba 2 0.0065 0.4048 1>)
       Scale Factor=&MakeVector4_FromLinearColor(LinearColor=&ColorFromCurve(ColorCurve=DataInterface() {propertyValues {
           AlphaCurve {keys []}
           BlueCurve {keys [<at 0.0632 10>, {arriveTangent=-9.0064, interpMode=RCIM_Cubic, time=0.1214, value=1}, <at 1 1 RCIM_Cubic>]}
           GreenCurve {keys [<at 0.0632 10>, {arriveTangent=-9.0064, interpMode=RCIM_Cubic, time=0.1214, value=1}, <at 1 1 RCIM_Cubic>]}
           RedCurve {keys [<at 0.0632 10>, {arriveTangent=-9.0064, interpMode=RCIM_Cubic, time=0.1214, value=1}, <at 1 1 RCIM_Cubic>]}
          }}))
      ))
     "Float (Alpha)"=&FloatFromCurve(FloatCurve=DataInterface() {propertyValues {Curve {keys [<at -0.0046 1.0113 RCIM_Cubic>,<at 1.0011 0.002 RCIM_Cubic>]}}})
    )
    Particles.11 - Sparkle Sprites Color:LinearColor=&MakeLinearColorFromVectorAndFloat(
     "Vector (RGB)"=&MakeVectorFromLinearColorRGB(Color=&MultiplyLinearColor(
       Linear Color=&RandomRangeLinearColor(Minimum=<rgba 8 5.1322 6.231 1>, Maximum=<rgba 8 1.6416 3.9094 1>)
       Scale Factor=&MakeVector4_FromLinearColor(LinearColor=&ColorFromCurve(ColorCurve=DataInterface() {propertyValues {
           AlphaCurve {keys []}
           BlueCurve {keys [<at 0.0632 10>, {arriveTangent=-9.0064, interpMode=RCIM_Cubic, time=0.1214, value=1}, <at 1 1 RCIM_Cubic>]}
           GreenCurve {keys [<at 0.0632 10>, {arriveTangent=-9.0064, interpMode=RCIM_Cubic, time=0.1214, value=1}, <at 1 1 RCIM_Cubic>]}
           RedCurve {keys [<at 0.0632 10>, {arriveTangent=-9.0064, interpMode=RCIM_Cubic, time=0.1214, value=1}, <at 1 1 RCIM_Cubic>]}
          }}))
      ))
     "Float (Alpha)"=&FloatFromCurve(FloatCurve=DataInterface() {propertyValues {Curve {keys [<at -0.0046 1.0113 RCIM_Cubic>,<at 1.0011 0.002 RCIM_Cubic>]}}})
    )
    Particles.9 - Heart Sprites COlor:LinearColor=&MakeLinearColorFromVectorAndFloat(
     "Vector (RGB)"=&MakeVectorFromLinearColorRGB(Color=&MultiplyLinearColor(
       Linear Color=&RandomRangeLinearColor(Minimum=<rgba 2 0.4583 1.181 1>, Maximum=<rgba 2 0.0273 0.4448 1>)
       Scale Factor=&MakeVector4_FromLinearColor(LinearColor=&ColorFromCurve(ColorCurve=DataInterface() {propertyValues {
           AlphaCurve {keys []}
           BlueCurve {keys [<at 0.0632 10>, {arriveTangent=-9.0064, interpMode=RCIM_Cubic, time=0.1214, value=1}, <at 1 1 RCIM_Cubic>]}
           GreenCurve {keys [<at 0.0632 10>, {arriveTangent=-9.0064, interpMode=RCIM_Cubic, time=0.1214, value=1}, <at 1 1 RCIM_Cubic>]}
           RedCurve {keys [<at 0.0632 10>, {arriveTangent=-9.0064, interpMode=RCIM_Cubic, time=0.1214, value=1}, <at 1 1 RCIM_Cubic>]}
          }}))
      ))
     "Float (Alpha)"=&FloatFromCurve(FloatCurve=DataInterface() {propertyValues {Curve {keys [<at -0.0046 1.0113 RCIM_Cubic>,<at 1.0011 0.002 RCIM_Cubic>]}}})
    )
    Particles.9-11 - Sprites Size:Vector2f=&Vector2DFromFloat(Value=&Multiply_Float(
      A=&RandomRangeFloat(Minimum=2, Maximum=15)
      B=&FloatFromCurve(FloatCurve=DataInterface() {propertyValues {Curve {keys [
           <at -0.0046 1.0113 RCIM_Cubic>
           {arriveTangent=-1.0036, interpMode=RCIM_Cubic, leaveTangent=-1.0036, time=0.7001, value=0.872}
           <at 1.0011 0.002>
          ]}}})
     ))
   ) {isSetParameters=true}
   &GravityForce(Gravity=&MaskVector(
     Mask Source=0
     Mask Parameter=?
     Mask=9
     Masked Vector=<0 0 -400>
     Passthrough Vector=&MaskVector(
      Mask Source=0
      Mask Parameter=?
      Mask=10
      Masked Vector=<0 0 -400>
      Passthrough Vector=&MaskVector(Mask Source=0, Mask Parameter=?, Mask=11, Masked Vector=<0 0 -400>, Passthrough Vector=<0 0 0>)
     )
    ))
   &Drag(
    Drag=&MaskFloat(
     Mask Source=0
     Mask Parameter=?
     Mask=9
     Masked Float=3
     Passthrough Float=&MaskFloat(
      Mask Source=0
      Mask Parameter=?
      Mask=10
      Masked Float=3
      Passthrough Float=&MaskFloat(Mask Source=0, Mask Parameter=?, Mask=11, Masked Float=3, Passthrough Float=0)
     )
    )
    Use Linear Drag=true
    Use Rotational Drag=false
   )
   &AccelerationForce(Acceleration=&MakeVector(
     X=0
     Y=0
     Z=&MaskFloat(
      Mask Source=0
      Mask Parameter=?
      Mask=9
      Masked Float=&Multiply_Float(
       A=10000
       B=&FloatFromCurve(FloatCurve=DataInterface() {propertyValues {Curve {keys [<at 0 0>,<at 0.0121 1.0118>,<at 0.0249 -0.0118>]}}})
      )
      Passthrough Float=&MaskFloat(
       Mask Source=0
       Mask Parameter=?
       Mask=10
       Masked Float=&Multiply_Float(
        A=10000
        B=&FloatFromCurve(FloatCurve=DataInterface() {propertyValues {Curve {keys [<at 0 0>,<at 0.0121 1.0118>,<at 0.0249 -0.0118>]}}})
       )
       Passthrough Float=&MaskFloat(
        Mask Source=0
        Mask Parameter=?
        Mask=11
        Masked Float=&Multiply_Float(
         A=10000
         B=&FloatFromCurve(FloatCurve=DataInterface() {propertyValues {Curve {keys [<at 0 0>,<at 0.0121 1.0118>,<at 0.0249 -0.0118>]}}})
        )
        Passthrough Float=0
       )
      )
     )
    ))
   &SolveForcesAndVelocity(Clamp Velocity=false, Limit Acceleration=false)
   &FN_Sprite_Attributes(
    Alignment:Vector3f=<0 0 1>
    Create Alignment Attribute=true
    Create Facing Attribute=true
    Facing:Vector3f=<0 1 0>
    Write to Module Namespace=true
   )
  ]
  rendererClasses [&NiagaraSpriteRendererProperties]
  renderers [
   {rendererClass=&NiagaraSpriteRendererProperties}
   {rendererClass=&NiagaraSpriteRendererProperties, rendererIndex=1}
   {rendererClass=&NiagaraSpriteRendererProperties, rendererIndex=2}
   {rendererClass=&NiagaraSpriteRendererProperties, rendererIndex=3}
   {rendererClass=&NiagaraSpriteRendererProperties, rendererIndex=4}
   {rendererClass=&NiagaraSpriteRendererProperties, rendererIndex=5}
   {rendererClass=&NiagaraSpriteRendererProperties, rendererIndex=6}
   {rendererClass=&NiagaraSpriteRendererProperties, rendererIndex=7}
   {rendererClass=&NiagaraSpriteRendererProperties, rendererIndex=8}
   {rendererClass=&NiagaraSpriteRendererProperties, rendererIndex=9}
   {rendererClass=&NiagaraSpriteRendererProperties, rendererIndex=10}
   {rendererClass=&NiagaraSpriteRendererProperties, rendererIndex=11}
   {rendererClass=&NiagaraSpriteRendererProperties, rendererIndex=12}
   {rendererClass=&NiagaraSpriteRendererProperties, rendererIndex=13}
  ]
 }
 CuddleTeamParticles {
  emitterUpdateScript [
   &EmitterState(Life Cycle Mode=::NewEnumerator1, Loop Behavior=::NewEnumerator1, Loop Duration=1.2, UseLoopDelay=false)
   &SpawnRate(
    SpawnRate=&Multiply_Float(
     A=10
     B=&FloatFromCurve(
      FloatCurve=DataInterface() {propertyValues {Curve {
         defaultValue=1
         keys [
          {arriveTangent=14.577, interpMode=RCIM_Cubic, leaveTangent=14.577, tangentMode=RCTM_User, time=0, value=0}
          {interpMode=RCIM_Cubic, tangentMode=RCTM_User, time=0.15, value=1}
          {interpMode=RCIM_Cubic, tangentMode=RCTM_User, time=1, value=0}
         ]
        }}}
      CurveIndex=$Emitter.NormalizedLoopAge
     )
    )
    Use Spawn Probability=false
   )
   &SpawnRate(
    SpawnGroup=1
    SpawnRate=&Multiply_Float(
     A=20
     B=&FloatFromCurve(
      FloatCurve=DataInterface() {propertyValues {Curve {
         defaultValue=1
         keys [
          {arriveTangent=14.577, interpMode=RCIM_Cubic, leaveTangent=14.577, tangentMode=RCTM_User, time=0, value=0}
          {interpMode=RCIM_Cubic, tangentMode=RCTM_User, time=0.15, value=1}
          {interpMode=RCIM_Cubic, tangentMode=RCTM_User, time=1, value=0}
         ]
        }}}
      CurveIndex=$Emitter.NormalizedLoopAge
     )
    )
    Use Spawn Probability=false
   ) {module=SpawnRate001}
   &SpawnRate(
    SpawnGroup=2
    SpawnRate=&Multiply_Float(
     A=30
     B=&FloatFromCurve(
      FloatCurve=DataInterface() {propertyValues {Curve {
         defaultValue=1
         keys [
          <at 0 0 RCIM_Cubic>
          {interpMode=RCIM_Cubic, tangentMode=RCTM_User, time=0.1016, value=1}
          {interpMode=RCIM_Cubic, tangentMode=RCTM_User, time=1, value=0}
         ]
        }}}
      CurveIndex=$Emitter.NormalizedLoopAge
     )
    )
    Use Spawn Probability=false
   ) {module=SpawnRate002}
   &SpawnRate(
    SpawnGroup=3
    SpawnRate=&Multiply_Float(
     A=50
     B=&FloatFromCurve(
      FloatCurve=DataInterface() {propertyValues {Curve {
         defaultValue=1
         keys [
          <at 0 0 RCIM_Cubic>
          {interpMode=RCIM_Cubic, tangentMode=RCTM_User, time=0.1016, value=1}
          {interpMode=RCIM_Cubic, tangentMode=RCTM_User, time=1, value=0}
         ]
        }}}
      CurveIndex=$Emitter.NormalizedLoopAge
     )
    )
    Use Spawn Probability=false
   ) {module=SpawnRate003}
  ]
  particleSpawnScript [
   &PartitionParticles()
   /SetVariables(Particles.VisibilityTag=&SpawnGroup()) {isSetParameters=true}
   &InitializeParticle(
    Color=<rgba 1 0.6966 0.8794 1>
    Lifetime Max=1.8
    Lifetime Min=0.25
    Lifetime Mode=::NewEnumerator1
    Mass Mode=::NewEnumerator3
    Sprite Rotation Mode=::NewEnumerator2
    Sprite Size Mode=::NewEnumerator1
    UsePositionOffset=false
    Write Material Random=true
   )
   &ShapeLocation(MaskSpawnGroup=false, Override Seed=false, Sphere Radius=3, UseOwnerScale=true)
   &SkeletalMeshLocation(
    Bone / Socket Index=$Particles.PartitionParticles.Partition
    Bone Sampling Mode=::NewEnumerator4
    MaskSpawnGroup=false
    Skeletal Mesh=$System.Skeletal Mesh
   )
  ]
  particleUpdateScript [
   &ParticleState()
   &ScaleSpriteSize(
    Scale Sprite Size Mode=::NewEnumerator0
    Uniform Scale Factor=&Multiply_Float(
     A=1
     B=&MakeFloatFromVector2d(Vector 2D=&Vector2DFromFloat(Value=&FloatFromCurve(
        FloatCurve=DataInterface() {propertyValues {Curve {keys [
            {interpMode=RCIM_Cubic, tangentMode=RCTM_User, time=0, value=0}
            {interpMode=RCIM_Cubic, tangentMode=RCTM_User, time=0.5, value=1}
            {tangentMode=RCTM_User, time=1, value=0}
           ]}}}
        Scale Curve=1.5
       )))
    )
   )
   &ScaleColor(
    Scale Alpha=&FloatFromCurve(FloatCurve=DataInterface() {propertyValues {Curve {keys [<at -0.0046 1.0113 RCIM_Cubic>,<at 1.0011 0.002 RCIM_Cubic>]}}})
    Scale RGB=&Multiply_Vector(
     A=<1 1 1>
     B=&MakeVectorFromLinearColorRGB(Color=&LerpLinearColors(
       StartColor=&MakeCustomLinearColorFromBool(
        Bool=&SetBoolByFloatComparison(A=&RandomRangeFloat(Minimum=0, Maximum=1), B=0.5)
        True Color=<rgba 1 1 1 1>
        False Color=<rgba 0 0 0 2>
       )
       EndColor=&SelectLinearColorFromArray(
        Array Sampling Mode=::NewEnumerator0
        Direct Array Index=?
        Color Selection Array=DataInterface() {propertyValues {ColorData [<rgba 0 0 0 2>, <rgba 1 0.1583 0.7693 1>, <rgba 1 0.0729 0.4249 1>, <rgba 0.7812 0 0.3477 1>, <rgba 1 0.7079 0.8931 1>]}}
        Random Seed=0
       )
       LerpFactor=&FloatFromCurve(FloatCurve=DataInterface() {propertyValues {Curve {keys [{interpMode=RCIM_Cubic, tangentMode=RCTM_User, time=0.1, value=0}, {interpMode=RCIM_Cubic, tangentMode=RCTM_User, time=0.6657, value=1}]}}})
      ))
    )
    ScaleA=true
    ScaleRGB=true
   )
   &CurlNoiseForce(
    Noise Strength=&FloatFromCurve(FloatCurve=DataInterface() {propertyValues {Curve {defaultValue=1, keys [<at 0.05 0 RCIM_Cubic>,<at 0.6 1 RCIM_Cubic>]}}}, Scale Curve=100)
    Pan Noise=false
   )
   &VortexForce(
    Origin Pull Amount:NiagaraFloat=&FloatFromCurve(FloatCurve=DataInterface() {propertyValues {Curve {keys [<at 0.15 1 RCIM_Cubic>,<at 0.5 -5 RCIM_Cubic>]}}})
    UseOriginOffset=false
    Vortex Force Amount=-200
   )
   &Drag(Drag=&RandomRangeFloat(Minimum=1, Maximum=5), Use Linear Drag=true, Use Rotational Drag=true)
   &PointForce(
    Force Strength=&FloatFromCurve(FloatCurve=DataInterface() {propertyValues {Curve {keys [<at 0 0>,<at 1 -100>]}}})
    Module.Use Falloff=false
    Module.UseLinearForce=false
   )
   &GravityForce(Gravity=&MakeVector(
     X=0
     Y=0
     Z=&FloatFromCurve(FloatCurve=DataInterface() {propertyValues {Curve {keys [<at 0 0>,<at 0.3084 10>,<at 1 -50>]}}})
    ))
   &SolveForcesAndVelocity(Clamp Velocity=false, Limit Acceleration=false)
   /SetVariables(
    Particles.Linear Color=&MakeLinearColorFromVectorAndFloat(
     "Vector (RGB)"=&MakeVectorFromLinearColorRGB(Color=&MultiplyLinearColor(
       Linear Color=&RandomRangeLinearColor(Minimum=<rgba 2 0.9062 1.4463 1>, Maximum=<rgba 2 0.0065 0.4048 1>)
       Scale Factor=&MakeVector4_FromLinearColor(LinearColor=&ColorFromCurve(ColorCurve=DataInterface() {propertyValues {
           AlphaCurve {keys []}
           BlueCurve {keys [<at 0.0632 10>, {arriveTangent=-9.0064, interpMode=RCIM_Cubic, time=0.1214, value=1}, <at 1 1 RCIM_Cubic>]}
           GreenCurve {keys [<at 0.0632 10>, {arriveTangent=-9.0064, interpMode=RCIM_Cubic, time=0.1214, value=1}, <at 1 1 RCIM_Cubic>]}
           RedCurve {keys [<at 0.0632 10>, {arriveTangent=-9.0064, interpMode=RCIM_Cubic, time=0.1214, value=1}, <at 1 1 RCIM_Cubic>]}
          }}))
      ))
     "Float (Alpha)"=&FloatFromCurve(FloatCurve=DataInterface() {propertyValues {Curve {keys [<at -0.0046 1.0113 RCIM_Cubic>,<at 1.0011 0.002 RCIM_Cubic>]}}})
    )
    Particles.Sparkles:LinearColor=&MakeLinearColorFromVectorAndFloat(
     "Vector (RGB)"=&MakeVectorFromLinearColorRGB(Color=&RandomRangeLinearColor(Minimum=<rgba 4 2.0449 3.0473 1>, Maximum=<rgba 5 0 1.6549 1>))
     "Float (Alpha)"=&FloatFromCurve(FloatCurve=DataInterface() {propertyValues {Curve {keys [<at -0.0046 1.0113 RCIM_Cubic>,<at 1.0011 0.002 RCIM_Cubic>]}}})
    )
    Particles.Sparkles Size:Vector2f=&Vector2DFromFloat(Value=&Multiply_Float(
      A=&RandomRangeFloat(Minimum=1, Maximum=9)
      B=&FloatFromCurve(FloatCurve=DataInterface() {propertyValues {Curve {
          defaultValue=1
          keys [
           {tangentMode=RCTM_User, time=0, value=1}
           {arriveTangent=-0.0405, interpMode=RCIM_Cubic, leaveTangent=-0.0405, tangentMode=RCTM_User, time=0.7009, value=1}
           {arriveTangent=-8.0089, interpMode=RCIM_Cubic, leaveTangent=-8.0089, tangentMode=RCTM_User, time=1, value=0}
          ]
         }}})
     ))
    Particles.Vector 2D:Vector2f=&Vector2DFromFloat(Value=&Multiply_Float(
      A=&RandomRangeFloat(Minimum=2, Maximum=8)
      B=&Multiply_Float(
       A=1
       B=&MakeFloatFromVector2d(Vector 2D=&Vector2DFromFloat(Value=&FloatFromCurve(
          FloatCurve=DataInterface() {propertyValues {Curve {keys [
              {interpMode=RCIM_Cubic, tangentMode=RCTM_User, time=0, value=0}
              {interpMode=RCIM_Cubic, tangentMode=RCTM_User, time=0.5, value=1}
              {tangentMode=RCTM_User, time=1, value=0}
             ]}}}
          Scale Curve=1.5
         )))
      )
     ))
   ) {isSetParameters=true}
  ]
  rendererClasses [&NiagaraSpriteRendererProperties]
  renderers [
   {rendererClass=&NiagaraSpriteRendererProperties}
   {rendererClass=&NiagaraSpriteRendererProperties, rendererIndex=1}
   {rendererClass=&NiagaraSpriteRendererProperties, rendererIndex=2}
   {rendererClass=&NiagaraSpriteRendererProperties, rendererIndex=3}
   {rendererClass=&NiagaraSpriteRendererProperties, rendererIndex=4}
  ]
 }
 FollowerTrail {
  emitterUpdateScript [
   &EmitterState(Life Cycle Mode=::NewEnumerator1, Loop Behavior=::NewEnumerator1, Loop Duration=3, UseLoopDelay=false)
   &SpawnParticlesFromOtherEmitter(
    Attribute Reader=DataInterface() {propertyValues {EmitterBinding {bindingMode=Other, emitterName=LeaderSprite}}}
    Spawn Rate=60
    Use Maximum Age=false
    Use Minimum Age=false
   )
  ]
  particleSpawnScript [
   &InitializeRibbon(
    Color=<rgba 1 1 1 3>
    Lifetime=2
    Position Mode:ENiagara_PositionInitializationMode=::NewEnumerator0
    Ribbon Facing Mode:ENiagara_UnsetDirectSet=::NewEnumerator1
    Ribbon Width=40
    Write Material Random=true
   )
   &InitializeParticle(UsePositionOffset=false, Write Material Random=true)
   &SampleParticlesFromOtherEmitter(
    Age Sampling:ENiagara_DirectReadSamplingMode=::NewEnumerator0
    Applied Position Offset:Vector3f=<-2 0 0>
    Color Sampling=::NewEnumerator1
    Particle ID Sampling=::NewEnumerator1
   )
   &AddVelocity(Velocity=<0 0 100>) {enabled=false}
   /SetVariables(Particles.Initial Detail Ribbon Width=25) {isSetParameters=true}
   &CameraOffset(Camera Offset Amount=1)
  ]
  particleUpdateScript [
   &ParticleState()
   &SolveForcesAndVelocity(Clamp Velocity=false, Limit Acceleration=false)
   &ScaleColor(
    Scale Alpha=&FloatFromCurve(
     FloatCurve=DataInterface() {propertyValues {Curve {keys [
         {arriveTangent=0.583, interpMode=RCIM_Cubic, leaveTangent=0.583, tangentMode=RCTM_User, time=0, value=0}
         {arriveTangent=0.3962, interpMode=RCIM_Cubic, leaveTangent=0.3962, tangentMode=RCTM_User, time=0.2458, value=0.9764}
         {arriveTangent=-1.7441, interpMode=RCIM_Cubic, leaveTangent=-1.7441, tangentMode=RCTM_User, time=0.4032, value=0.8057}
         {arriveTangent=0.0107, interpMode=RCIM_Cubic, leaveTangent=0.0107, tangentMode=RCTM_User, time=0.9851, value=0.0118}
        ]}}}
     CurveIndex=$Emitter.Age
     Scale Curve=0.2
    )
    Scale RGB=&MakeVectorFromLinearColorRGB(Color=&ColorFromCurve(ColorCurve=DataInterface() {propertyValues {
        AlphaCurve {keys [<at 0 0>,<at 0.4883 0.2>,<at 0.9729 0>]}
        BlueCurve {keys [
          <at 0.0022 0.2865>
          <at 0.2151 0.5989>
          <at 0.3547 0.0899>
          {arriveTangent=4.8038, interpMode=RCIM_Cubic, leaveTangent=4.8038, time=0.5045, value=0.3499}
          <at 0.5882 0.5641>
          <at 0.6572 1.5432>
          <at 0.7086 0.3013>
          <at 0.7938 0.5156>
          <at 0.8819 0.2865>
         ]}
        GreenCurve {keys [
          <at 0.0022 0>
          <at 0.2151 0.2083>
          <at 0.3547 0>
          {arriveTangent=0.8608, interpMode=RCIM_Cubic, leaveTangent=0.8608, time=0.5045, value=0.0138}
          <at 0.5882 0.0065>
          <at 0.6572 0.2604>
          <at 0.7086 0.0171>
          <at 0.7938 0>
          <at 0.8819 0>
         ]}
        RedCurve {keys [
          <at 0.0022 0.1205>
          <at 0.2151 1>
          <at 0.3547 0.276>
          {arriveTangent=5.6985, interpMode=RCIM_Cubic, leaveTangent=5.6985, time=0.5045, value=0.4427}
          <at 0.5882 0.625>
          <at 0.6572 2>
          <at 0.7086 0.4115>
          <at 0.7938 0.3653>
          <at 0.8819 0.1205>
         ]}
       }}))
    ScaleA=true
    ScaleRGB=true
   )
   &ScaleColor(
    Scale Alpha=0.7
    Scale RGB=&VectorFromFloat(Value=&FNTimeOfDayScalar(Night Scale=25, Morning Scale=1, Day Scale=25, Evening Scale=1))
    ScaleA=false
    ScaleRGB=true
   ) {module=ScaleColor001}
   &ScaleRibbonWidth(Ribbon Width Scale=&FloatFromCurve(FloatCurve=DataInterface() {propertyValues {Curve {
        defaultValue=0.5
        keys [{interpMode=RCIM_Cubic, tangentMode=RCTM_User, time=0, value=1}, {interpMode=RCIM_Cubic, tangentMode=RCTM_User, time=1, value=0}]
       }}}))
   /SetVariables(Particles.Secondary Ribbon Color:LinearColor=&ColorFromCurve(ColorCurve=DataInterface() {propertyValues {
       AlphaCurve {keys [<at 0 0>,<at 0.4685 0.7 RCIM_Cubic>,<at 1 0 RCIM_Cubic>]}
       BlueCurve {keys [<at 0 0.1336 RCIM_Cubic>,<at 0.0696 0.2056>,<at 0.1901 0.3769>,<at 0.3695 0.796>,<at 0.6378 0.3595>,<at 0.8401 0.2131>,<at 0.9477 0.0712 RCIM_Cubic>]}
       GreenCurve {keys [<at 0 0 RCIM_Cubic>,<at 0.0696 0>,<at 0.1901 0.0323>,<at 0.3695 0.2679>,<at 0.6378 0.0289>,<at 0.8401 0>,<at 0.9477 0.0005 RCIM_Cubic>]}
       RedCurve {keys [<at 0 0.2969 RCIM_Cubic>,<at 0.0696 0.5365>,<at 0.1901 0.8854>,<at 0.3695 0.8699>,<at 0.6378 0.8401>,<at 0.8401 0.4583>,<at 0.9477 0.1615 RCIM_Cubic>]}
      }})) {isSetParameters=true}
   /SetVariables(Particles.Update Detail Ribbon Width:NiagaraFloat=&Multiply_Float(
     A=$Particles.Initial.Initial Detail Ribbon Width
     B=&FloatFromCurve(FloatCurve=DataInterface() {propertyValues {Curve {
         defaultValue=0.5
         keys [{interpMode=RCIM_Cubic, tangentMode=RCTM_User, time=0, value=1}, {interpMode=RCIM_Cubic, tangentMode=RCTM_User, time=1, value=0}]
        }}})
    )) {isSetParameters=true}
  ]
  rendererClasses [&NiagaraRibbonRendererProperties]
  renderers [
   {rendererClass=&NiagaraRibbonRendererProperties}
   {rendererClass=&NiagaraRibbonRendererProperties, rendererIndex=1}
   {rendererClass=&NiagaraRibbonRendererProperties, rendererIndex=2}
  ]
 }
 HandRibbons {
  emitterUpdateScript [
   &EmitterState(Life Cycle Mode=::NewEnumerator1, Loop Behavior=::NewEnumerator1, Loop Duration=0.95, UseLoopDelay=false)
   &SpawnBurst_Instantaneous(Spawn Count=2, Use Spawn Probability=false, UseLoopCountLimit=false)
   &SpawnRate(
    SpawnRate=&Multiply_Float(
     A=150
     B=&FloatFromCurve(
      FloatCurve=DataInterface() {propertyValues {Curve {
         defaultValue=1
         keys [
          {interpMode=RCIM_Cubic, tangentMode=RCTM_User, time=0.3928, value=1}
          {arriveTangent=-5.3983, interpMode=RCIM_Cubic, leaveTangent=-5.3983, tangentMode=RCTM_User, time=0.8407, value=0}
         ]
        }}}
      CurveIndex=$Emitter.NormalizedLoopAge
     )
    )
    Use Spawn Probability=false
   )
  ]
  particleSpawnScript [
   &PartitionParticles()
   &InitializeParticle(
    Color=<rgba 8 8 8 1.5>
    Lifetime=0.7
    Ribbon Width=15
    Ribbon Width Mode=::NewEnumerator1
    UsePositionOffset=false
    Write Material Random=true
   )
   &SkeletalMeshLocation(
    Bone / Socket Index=$Particles.PartitionParticles.Partition
    Bone Sampling Mode=::NewEnumerator4
    MaskSpawnGroup=false
    Sampled Position Offset=&MaskVector(Mask Source=1, Mask Parameter=$Particles.PartitionParticles.Partition, Mask=0, Masked Vector=<2 0 -5>, Passthrough Vector=<-2 0 -5>)
    Skeletal Mesh=$System.Skeletal Mesh
   )
   &AddVelocity() {enabled=false}
  ]
  particleUpdateScript [
   &ParticleState()
   &InheritSourceMovement(ApplyForce=false, ApplyPosition=true, ApplyPositionFalloff=false)
   &Drag(Drag=5, Use Linear Drag=true, Use Rotational Drag=true)
   &SolveForcesAndVelocity(Clamp Velocity=false, Limit Acceleration=false)
   &ScaleRibbonWidth(Ribbon Width Scale=&FloatFromCurve(FloatCurve=DataInterface() {propertyValues {Curve {keys [<at 0 1>,<at 1 0.4>]}}}))
   &ScaleColor(
    Scale Alpha=&FloatFromCurve(
     FloatCurve=DataInterface() {propertyValues {Curve {keys [{arriveTangent=-3.0806, interpMode=RCIM_Cubic, leaveTangent=-3.0806, tangentMode=RCTM_User, time=0, value=1}, <at 1 0 RCIM_Cubic>]}}}
     Scale Curve=0.4
    )
    Scale RGB=&MakeVectorFromLinearColorRGB(Color=&ColorFromCurve(ColorCurve=DataInterface() {propertyValues {
        AlphaCurve {keys []}
        BlueCurve {keys [<at 0.1031 1.3773 RCIM_Cubic>,<at 0.3962 0.3257>,<at 1 0.2396>]}
        GreenCurve {keys [<at 0.1031 0.3672 RCIM_Cubic>,<at 0.3962 0>,<at 1 0>]}
        RedCurve {keys [<at 0.1031 1.5 RCIM_Cubic>,<at 0.3962 1>,<at 1 1>]}
       }}))
    ScaleA=true
    ScaleRGB=true
   )
  ]
  rendererClasses [&NiagaraRibbonRendererProperties]
  renderers [
   {rendererClass=&NiagaraRibbonRendererProperties}
   {rendererClass=&NiagaraRibbonRendererProperties, rendererIndex=1}
  ]
 }
 LeaderSprite {
  emitterUpdateScript [
   &EmitterState(Life Cycle Mode=::NewEnumerator1, Loop Behavior=::NewEnumerator1, Loop Duration=3, UseLoopDelay=false)
   &SpawnBurst_Instantaneous(Spawn Count=2, Spawn Time=0.2, Use Spawn Probability=false, UseLoopCountLimit=false)
   &SpawnRate(
    SpawnRate=&Multiply_Float(
     A=6
     B=&FloatFromCurve(
      FloatCurve=DataInterface() {propertyValues {Curve {keys [
          {arriveTangent=-7.1176, interpMode=RCIM_Cubic, leaveTangent=-7.1176, tangentMode=RCTM_User, time=0.2235, value=1}
          {arriveTangentWeight=0.2088, interpMode=RCIM_Cubic, tangentMode=RCTM_User, time=0.65, value=0}
         ]}}}
      CurveIndex=$Emitter.NormalizedLoopAge
     )
    )
    Use Spawn Probability=false
   ) {enabled=false, module=SpawnRate001}
  ]
  particleSpawnScript [
   &PartitionParticles()
   &InitializeParticle(
    Lifetime=0.7
    Sprite Size Mode=::NewEnumerator1
    Uniform Sprite Size Max=0
    Uniform Sprite Size Min=0
    UsePositionOffset=false
    Write Material Random=true
   ) {module=InitializeParticle001}
  ]
  particleUpdateScript [
   &ParticleState()
   &SystemLocation(
    MaskSpawnGroup=false
    Offset=&MaskVector(
     Mask Source=1
     Mask Parameter=$Particles.PartitionParticles.Partition
     Mask=0
     Masked Vector=&MakeVector(
      X=&Add_Float(
       A=12
       B=&FloatFromCurve(
        FloatCurve=DataInterface() {propertyValues {Curve {keys [
            {arriveTangent=1.0335, interpMode=RCIM_Cubic, leaveTangent=1.0335, tangentMode=RCTM_User, time=0, value=0}
            {arriveTangent=-5.7115, interpMode=RCIM_Cubic, leaveTangent=-5.7115, tangentMode=RCTM_User, time=0.1176, value=-0.5241}
            {arriveTangent=7.196, interpMode=RCIM_Cubic, leaveTangent=7.196, tangentMode=RCTM_User, time=0.1998, value=-0.5078}
            {arriveTangent=3.0037, interpMode=RCIM_Cubic, leaveTangent=3.0037, tangentMode=RCTM_User, time=0.3013, value=0.1474}
            {arriveTangent=-3.2136, interpMode=RCIM_Cubic, leaveTangent=-3.2136, tangentMode=RCTM_User, time=0.5783, value=-0.0795}
            {arriveTangent=-0.699, interpMode=RCIM_Cubic, leaveTangent=-0.699, time=0.7884, value=-0.6149}
            {arriveTangent=3.0036, interpMode=RCIM_Cubic, leaveTangent=3.0036, tangentMode=RCTM_User, time=1.0145, value=-0.3844}
           ]}}}
        Scale Curve=22
       )
      )
      Y=0
      Z=&Add_Float(
       A=13
       B=&FloatFromCurve(
        FloatCurve=DataInterface() {propertyValues {Curve {keys [
            <at 0 -0.4 RCIM_Cubic>
            {arriveTangent=4.9183, interpMode=RCIM_Cubic, leaveTangent=4.9183, tangentMode=RCTM_User, time=0.3485, value=1.2189}
            {arriveTangent=-0.5694, interpMode=RCIM_Cubic, leaveTangent=-0.5694, tangentMode=RCTM_User, time=0.61, value=1.7906}
            {arriveTangent=-4.507, interpMode=RCIM_Cubic, leaveTangent=-4.507, tangentMode=RCTM_User, time=0.7235, value=1.4359}
            {arriveTangent=-1.2279, interpMode=RCIM_Cubic, leaveTangent=-1.2279, time=0.9176, value=1.0857}
            <at 0.9963 1.1009 RCIM_Cubic>
           ]}}}
        Scale Curve=22
       )
      )
     )
     Passthrough Vector=&MakeVector(
      X=&Add_Float(
       A=-8
       B=&FloatFromCurve(
        FloatCurve=DataInterface() {propertyValues {Curve {keys [
            {arriveTangent=1.0335, interpMode=RCIM_Cubic, leaveTangent=1.0335, tangentMode=RCTM_User, time=0, value=0}
            {arriveTangent=-5.7115, interpMode=RCIM_Cubic, leaveTangent=-5.7115, tangentMode=RCTM_User, time=0.1176, value=-0.5241}
            {arriveTangent=7.196, interpMode=RCIM_Cubic, leaveTangent=7.196, tangentMode=RCTM_User, time=0.1998, value=-0.5078}
            {arriveTangent=3.0037, interpMode=RCIM_Cubic, leaveTangent=3.0037, tangentMode=RCTM_User, time=0.3013, value=0.1474}
            {arriveTangent=-3.2136, interpMode=RCIM_Cubic, leaveTangent=-3.2136, tangentMode=RCTM_User, time=0.5783, value=-0.0795}
            {arriveTangent=-0.699, interpMode=RCIM_Cubic, leaveTangent=-0.699, time=0.7884, value=-0.6149}
            {arriveTangent=3.0036, interpMode=RCIM_Cubic, leaveTangent=3.0036, tangentMode=RCTM_User, time=1.0145, value=-0.3844}
           ]}}}
        Scale Curve=-22
       )
      )
      Y=0
      Z=&Add_Float(
       A=13
       B=&FloatFromCurve(
        FloatCurve=DataInterface() {propertyValues {Curve {keys [
            <at 0 -0.4 RCIM_Cubic>
            {arriveTangent=4.9183, interpMode=RCIM_Cubic, leaveTangent=4.9183, tangentMode=RCTM_User, time=0.3485, value=1.2189}
            {arriveTangent=-0.5694, interpMode=RCIM_Cubic, leaveTangent=-0.5694, tangentMode=RCTM_User, time=0.61, value=1.7906}
            {arriveTangent=-4.507, interpMode=RCIM_Cubic, leaveTangent=-4.507, tangentMode=RCTM_User, time=0.7235, value=1.4359}
            {arriveTangent=-1.2279, interpMode=RCIM_Cubic, leaveTangent=-1.2279, time=0.9176, value=1.0857}
            <at 0.9963 1.1009 RCIM_Cubic>
           ]}}}
        Scale Curve=22
       )
      )
     )
    )
   )
  ]
  rendererClasses [&NiagaraSpriteRendererProperties]
  renderers [{rendererClass=&NiagaraSpriteRendererProperties}]
 }
}
summary {
 systemName=NS_Emote_RoseBruin_Trails
 emitters [
  {emitterName=CuddleTeamParticles, rendererClasses [&NiagaraSpriteRendererProperties]}
  {emitterName=LeaderSprite, rendererClasses [&NiagaraSpriteRendererProperties]}
  {emitterName=FollowerTrail, rendererClasses [&NiagaraRibbonRendererProperties]}
  {emitterName=HandRibbons, rendererClasses [&NiagaraRibbonRendererProperties]}
  {emitterName=Beam, rendererClasses [&NiagaraMeshRendererProperties, &NiagaraSpriteRendererProperties]}
  {emitterName=Condensed_Meshes, rendererClasses [&NiagaraMeshRendererProperties]}
  {emitterName=Condensed_Sprites, rendererClasses [&NiagaraSpriteRendererProperties]}
 ]
 userVariables [
  User.bIsFrontEnd {type=NiagaraBool}
  User.bIsFrontEndPreview {type=NiagaraBool}
  User.FrontEndScale {type=NiagaraFloat, defaultValue=0.6, description=""}
 ]
}
